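# Niagara sparks effect for the RockClimb emote, emitting glowing sparks, streaks and smoke from skeletal mesh bones or sockets over a timed curve.
# Condensed_Sparks merges eight layers through spawn groups and MergeHelper switches, with disabled standalone Base_Streaks and Classic_Sparks emitters matching two layers.
# A separate BouncingSparks emitter adds brief flecks scaled by User.Intensity.
# Condensed_Sparks uses vortex force, flicker waveforms and owner-speed-dependent world/local blending.
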NiagaraSystem /BRCosmetics/Animation/Game/MainPlayer/Emotes/RockClimb/VFX/NS_Emote_RockClimb_Sparks
dependencies {
 usedDataInterfaces [
  &NiagaraDataInterfaceCurve
  &NiagaraDataInterfaceVector2DCurve
  &NiagaraDataInterfaceColorCurve
  &NiagaraDataInterfaceSpriteRendererInfo
  &NiagaraDataInterfaceVectorCurve
  &NiagaraDataInterfaceSkeletalMesh
 ]
 usedDynamicInputs [&Multiply_Vector2D_ByFloat, &MergeHelper_ColorSwitch]
 usedModules [
  &self:SystemSpawnScript.SystemScriptSource.SystemScriptGraph.NiagaraNodeAssignment_0.SetVariables_#1
  &self:SystemSpawnScript.SystemScriptSource.SystemScriptGraph.NiagaraNodeAssignment_2.SetVariables_#2
  &self:SystemSpawnScript.SystemScriptSource.SystemScriptGraph.NiagaraNodeAssignment_1.SetVariables_#3
  &SystemState
  &BlendSpaceHelper
 ]
}
emitters {
 Base_Streaks {
  bEnabled=false
  emitterUpdateScript [
   &EmitterState(Inactive Response=::NewEnumerator1, Life Cycle Mode=::NewEnumerator1, Loop Behavior=::NewEnumerator1, Loop Duration=6.63, UseLoopDelay=false)
   &SpawnRate(
    SpawnRate=&FloatFromCurve(
     FloatCurve=DataInterface() {propertyValues {Curve {keys [
         {interpMode=RCIM_Cubic, tangentMode=RCTM_User, time=0.7148, value=40}
         {arriveTangent=-118.7953, interpMode=RCIM_Cubic, leaveTangent=-118.7953, tangentMode=RCTM_User, time=0.9617, value=-1.5686}
        ]}}}
     CurveIndex=$Emitter.NormalizedLoopAge
    )
    Use Spawn Probability=true
   ) {module=SpawnRate002}
  ]
  particleSpawnScript [
   &PartitionParticles()
   &InitializeParticle(
    Lifetime=&RandomRangeFloat(Minimum=0.1, Maximum=0.13)
    Sprite Size Mode=::NewEnumerator1
    Uniform Sprite Size Min=6
    UsePositionOffset=false
    Write Material Random=true
   )
   &SkeletalMeshLocation(
    Bone / Socket Index=$Particles.PartitionParticles.Partition
    MaskSpawnGroup=false
    Mesh Sampling Type=::NewEnumerator5
    Skeletal Mesh=$System.Skeletal Mesh
    Socket Sampling Mode=::NewEnumerator5
   )
   &VortexForce(
    Origin Pull Amount=-30
    UseOriginOffset=false
    Vortex Axis=<1 0 0>
    Vortex Force Amount=&RandomRangeFloat(Minimum=355, Maximum=370)
    Vortex Origin:NiagaraPosition=&AddVectorToPosition(Position=$Engine.Emitter.SimulationPosition, Vector=<0 0 130>)
   )
   &ApplyInitialForces()
  ]
  particleUpdateScript [
   &SolveForcesAndVelocity(Clamp Velocity=false, Limit Acceleration=false)
   &ParticleState()
   &ScaleSpriteSize(
    Scale Factor=&Multiply_Vector2D(
     A=&RandomRangeVector2D(Minimum=<2 2>, Maximum=<3 3>)
     B=&Vector2DFromCurve(Vector2Curve=DataInterface() {propertyValues {XCurve {keys [<at 0 1>,<at 1 0.5>]}, YCurve {keys [<at 0 1>,<at 1 0.5>]}}}, ScaleCurve=<1 1>)
    )
    Scale Sprite Size Mode=::NewEnumerator2
   )
   &ScaleColor(
    Linear Color Curve=DataInterface() {propertyValues {
      AlphaCurve {keys [<at 0 0.7>,<at 0.6507 0.18>,<at 1 0>]}
      BlueCurve {keys [<at 0.0424 2.6875>,<at 0.3289 0>,<at 0.6955 0>,<at 1 0>]}
      GreenCurve {keys [<at 0.0424 2.919>,<at 0.3289 1.3329>,<at 0.6955 0.3963>,<at 1 0.0724>]}
      RedCurve {keys [<at 0.0424 3>,<at 0.3289 2.5942>,<at 0.6955 1.8651>,<at 1 1>]}
     }}
    Scale Mode=::NewEnumerator2
   )
   &ScaleSpriteSizeBySpeed(Max Scale Factor=<1.15 6>, Min Scale Factor=<0.75 1>)
   &BlendSpace("Position Blend (World <-> Local)":NiagaraFloat=&MakeCustomFloatFromBool(
     Bool=&SetBoolByFloatComparison(A=&VectorLength(Vector=$Engine.Owner.Velocity), B=301)
     True Float=1
     False Float=&FloatFromCurve(FloatCurve=DataInterface() {propertyValues {Curve {
         defaultValue=0
         keys [{interpMode=RCIM_Cubic, tangentMode=RCTM_User, time=0, value=1}, {interpMode=RCIM_Cubic, tangentMode=RCTM_User, time=0.7, value=0.2}]
        }}})
     Translation Contribution:NiagaraFloat=1
     Rotation Contribution:NiagaraFloat=1
    )) {module=BlendSpace001}
  ]
  rendererClasses [&NiagaraSpriteRendererProperties]
  renderers [{rendererClass=&NiagaraSpriteRendererProperties}]
 }
 BouncingSparks {
  emitterUpdateScript [
   &EmitterState(Life Cycle Mode=::NewEnumerator1, Loop Behavior=::NewEnumerator1, Loop Duration=6.64, UseLoopDelay=false)
   &SpawnRate(
    SpawnRate=&Multiply_Float(
     A=&FloatFromCurve(
      FloatCurve=DataInterface() {propertyValues {Curve {keys [
          {arriveTangent=8251.3125, interpMode=RCIM_Cubic, leaveTangent=8251.3125, tangentMode=RCTM_User, time=0, value=0}
          {interpMode=RCIM_Cubic, tangentMode=RCTM_User, time=0.043, value=300}
          {interpMode=RCIM_Cubic, tangentMode=RCTM_User, time=0.8581, value=300}
          {interpMode=RCIM_Cubic, tangentMode=RCTM_User, time=0.953, value=0}
         ]}}}
      CurveIndex=$Emitter.NormalizedLoopAge
      Scale Curve=3
     )
     B=$User.Intensity
    )
    Use Spawn Probability=false
   )
  ]
  particleSpawnScript [
   &InitializeParticle(
    Color Channel Mode:ENiagaraChannelCorrelation=::NewEnumerator2
    Color Maximum=<rgba 6 2.3585 0.7944 0.25>
    Color Minimum=<rgba 1 0 0 0.25>
    Color Mode=::NewEnumerator2
    Lifetime=&RandomRangeFloat(Minimum=0.005, Maximum=0.025, Evaluation Type=::NewEnumerator1)
    Sprite Rotation Mode=::NewEnumerator2
    Sprite Size Mode=::NewEnumerator1
    Uniform Sprite Size Max=9
    Uniform Sprite Size Min=0.5
    UsePositionOffset=false
    Write Material Random=true
   )
   &PartitionParticles()
   &SkeletalMeshLocation(
    Bone / Socket Index=$Particles.PartitionParticles.Partition
    MaskSpawnGroup=false
    Mesh Sampling Type=::NewEnumerator5
    Skeletal Mesh=$System.Skeletal Mesh
    Socket Sampling Mode=::NewEnumerator5
   )
   &ShapeLocation(
    MaskSpawnGroup=false
    Override Seed=false
    Sphere Radius=&RandomRangeFloat(Minimum=6, Maximum=20, Evaluation Type=::NewEnumerator1)
    Sphere Surface Distribution=1
    UseOwnerScale=true
   )
   /CascadeToNiagaraConverter/NiagaraScripts/ModuleScripts/FixShapeVector() {enabled=false}
   &AddVelocity(Module.Use Falloff Distance=false, Override Seed=false, Velocity Mode=::NewEnumerator1)
  ]
  particleUpdateScript [
   &ParticleState()
   &ScaleSpriteSize(
    Scale Factor=&Vector2DFromFloat(Value=&FloatFromCurve(FloatCurve=DataInterface() {propertyValues {Curve {keys [<at 0 0>,<at 1 1>]}}}))
    Scale Sprite Size Mode=::NewEnumerator2
   )
   &SubUVAnimation(
    End Frame Range Override=2
    "Sprite Renderer "=DataInterface() {propertyValues {SpriteRenderer=&self:NiagaraEmitter_6.NiagaraSpriteRendererProperties_2}}
    UseEndFrame=false
    UseStartFrame=false
   ) {enabled=false}
   &Drag(Drag=3, Use Linear Drag=true, Use Rotational Drag=true)
   &SolveForcesAndVelocity(Clamp Velocity=false, Limit Acceleration=false)
   &ScaleColor(
    Scale Alpha=&FloatFromCurve(FloatCurve=DataInterface() {propertyValues {Curve {
        defaultValue=1
        keys [
         {tangentMode=RCTM_User, time=0, value=1}
         {arriveTangent=-0.0405, interpMode=RCIM_Cubic, leaveTangent=-0.0405, tangentMode=RCTM_User, time=0.7009, value=1}
         {arriveTangent=-8.0089, interpMode=RCIM_Cubic, leaveTangent=-8.0089, tangentMode=RCTM_User, time=1, value=0}
        ]
       }}})
    Scale RGB=&MakeVectorFromLinearColorRGB(Color=&MultiplyLinearColor(
      Linear Color=$Particles.Initial.Color
      Scale Factor=&MakeVector4_FromLinearColor(LinearColor=&RandomRangeLinearColor(Minimum=<rgba 5 1.3414 0.9492 1>, Maximum=<rgba 30 30 30 1>))
     ))
    ScaleA=true
    ScaleRGB=true
   )
  ]
  rendererClasses [&NiagaraSpriteRendererProperties]
  renderers [{rendererClass=&NiagaraSpriteRendererProperties}]
 }
 Classic_Sparks {
  bEnabled=false
  simTarget=GPUComputeSim
  emitterUpdateScript [
   &EmitterState(
    Enable Distance Culling=true
    Life Cycle Mode=::NewEnumerator1
    Loop Behavior=::NewEnumerator1
    Loop Delay=2.3833
    Loop Duration=6.64
    MaxDistance=7000
    Module.bUseMaxDistance=true
    Module.bUseMinDistance=false
    Scalability Mode=::NewEnumerator1
    UseLoopDelay=false
   )
   &SpawnRate(
    Spawn Probability=&FN_FrontEndFloatModifier(Gameplay=1, Front End=0)
    SpawnRate=&FloatFromCurve(
     FloatCurve=DataInterface() {propertyValues {Curve {keys [
         {interpMode=RCIM_Cubic, tangentMode=RCTM_User, time=0, value=0}
         {interpMode=RCIM_Cubic, tangentMode=RCTM_User, time=0.0277, value=170}
         {arriveTangent=-260.7543, interpMode=RCIM_Cubic, leaveTangent=-260.7543, tangentMode=RCTM_User, time=0.7257, value=104.7049}
         {interpMode=RCIM_Cubic, tangentMode=RCTM_User, time=1, value=0}
        ]}}}
     CurveIndex=$Emitter.NormalizedLoopAge
    )
    Use Spawn Probability=true
   )
  ]
  particleSpawnScript [
   &PartitionParticles()
   &InitializeParticle(
    Color Mode=::NewEnumerator0
    Lifetime Max=0.22
    Lifetime Min=0.13
    Lifetime Mode=::NewEnumerator1
    Sprite Rotation Mode=::NewEnumerator2
    Sprite Size Max=<3 6>
    Sprite Size Min=<1 1>
    Sprite Size Mode=::NewEnumerator2
    UsePositionOffset=false
    Write Material Random=true
   )
   &SkeletalMeshLocation(
    Bone / Socket Index=$Particles.PartitionParticles.Partition
    MaskSpawnGroup=false
    Mesh Sampling Type=::NewEnumerator5
    Skeletal Mesh=$System.Skeletal Mesh
    Socket Sampling Mode=::NewEnumerator5
   )
   &VortexForce(
    Origin Pull Amount=-30
    UseOriginOffset=false
    Vortex Axis=<1 0 0>
    Vortex Force Amount=&RandomRangeFloat(Minimum=355, Maximum=370)
    Vortex Origin:NiagaraPosition=&AddVectorToPosition(Position=$Engine.Emitter.SimulationPosition, Vector=<0 0 130>)
   )
   &ApplyInitialForces()
  ]
  particleUpdateScript [
   &ParticleState()
   /SetVariables(
    Particles.Flicker:NiagaraFloat=&Waveform("Waveform [1]"=::NewEnumerator0, "[1] Frequency"=75, "[1] Amplitude Min/Max"=<0 2>)
    Particles.Flicker Size:NiagaraFloat=&Waveform("Waveform [1]"=::NewEnumerator0, "[1] Frequency"=75, "[1] Amplitude Min/Max"=<0.15 1>)
    Particles.Flicker Value:NiagaraFloat=&Waveform("Waveform [1]"=::NewEnumerator0, "[1] Frequency"=75, "[1] Amplitude Min/Max"=<2 8>)
   ) {isSetParameters=true}
   &ScaleSpriteSize(
    Scale Sprite Size Mode=::NewEnumerator0
    Uniform Scale Factor=&Lerp_Float(
     A=1
     B=$Particles.Flicker Size
     Alpha=&FloatFromCurve(FloatCurve=DataInterface() {propertyValues {Curve {keys [<at 0 0>,<at 0.2485 0.4244>,<at 0.45 1>,<at 1 1>]}}})
    )
   )
   &ScaleColor(
    Scale Mode=::NewEnumerator1
    Scale RGBA=&Multiply_Vector4(
     A=&MakeVector4_FromLinearColor(LinearColor=&ColorFromCurve(ColorCurve=DataInterface() {propertyValues {
         AlphaCurve {keys [<at 0.122 1>]}
         BlueCurve {keys [<at 0 3.0372>,<at 0.1056 0>,<at 0.2341 0>,<at 0.8178 0>,<at 0.8544 0>,<at 0.9314 0>,<at 0.9928 0>]}
         GreenCurve {keys [<at 0 7.61>,<at 0.1056 3.3715>,<at 0.2341 0.7094>,<at 0.8178 0.1839>,<at 0.8544 5.7311>,<at 0.9314 0.6262>,<at 0.9928 0.2902>]}
         RedCurve {keys [<at 0 10>,<at 0.1056 10>,<at 0.2341 7.5>,<at 0.8178 7.5>,<at 0.8544 8.0224>,<at 0.9314 10>,<at 0.9928 10>]}
        }}))
     B=&MakeVector4_FromLinearColor(LinearColor=&MakeLinearColorFromVectorAndFloat(
       "Vector (RGB)"=&Add_Vector(
        A=<1 1 1>
        B=/Niagara/DynamicInputs/TypeConversions/MakeVectorFromLinearColorHSV(Color:LinearColor=/Niagara/DynamicInputs/LinearColor/MakeLinearColorFromHSV(
          HSV From:NiagaraInt32=1
          HSV:Vector3f=?
          Hue:NiagaraFloat=0
          Saturation:NiagaraFloat=0
          Value:NiagaraFloat=&Lerp_Float(
           A=0
           B=$Particles.Flicker Value
           Alpha=&FloatFromCurve(FloatCurve=DataInterface() {propertyValues {Curve {keys [<at 0 0>,<at 0.5 0>,<at 0.55 0.9624>,<at 1 1>]}}})
          )
          Alpha:NiagaraFloat=1
         ))
       )
       "Float (Alpha)"=&Lerp_Float(
        A=1
        B=$Particles.Flicker
        Alpha=&FloatFromCurve(FloatCurve=DataInterface() {propertyValues {Curve {keys [<at 0 0>,<at 0.25 0>,<at 0.35 1>,<at 1 1>]}}})
       )
      ))
    )
    ScaleRGBA=true
   )
   &ScaleColor(
    Linear Color Curve=DataInterface() {propertyValues {
      AlphaCurve {keys [<at 0.122 1>,<at 0.4022 0.35>,<at 0.6415 0.1>,<at 1 0>]}
      BlueCurve {keys [<at 0 3.0372>,<at 0.1056 0>,<at 0.2341 0>,<at 1 0>]}
      GreenCurve {keys [<at 0 7.61>,<at 0.1056 3.3715>,<at 0.2341 0.7094>,<at 1 0.1839>]}
      RedCurve {keys [<at 0 10>,<at 0.1056 10>,<at 0.2341 7.5>,<at 1 7.5>]}
     }}
    Scale Mode=::NewEnumerator2
   ) {enabled=false, module=ScaleColor001}
   &BlendSpace("Position Blend (World <-> Local)":NiagaraFloat=&MakeCustomFloatFromBool(
     Bool=&SetBoolByFloatComparison(A=&VectorLength(Vector=$Engine.Owner.Velocity), B=301)
     True Float=1
     False Float=&FloatFromCurve(FloatCurve=DataInterface() {propertyValues {Curve {
         defaultValue=0
         keys [{interpMode=RCIM_Cubic, tangentMode=RCTM_User, time=0, value=1}, {interpMode=RCIM_Cubic, tangentMode=RCTM_User, time=0.7, value=0.2}]
        }}})
     Translation Contribution:NiagaraFloat=1
     Rotation Contribution:NiagaraFloat=1
    )) {module=BlendSpace001}
   &InheritSourceMovement(Applied Position Scale=<0.15 0.15 0.15>, ApplyForce=false, ApplyPosition=true, ApplyPositionFalloff=false) {module=InheritSourceMovement001}
   &Drag(Drag=&Lerp_Float(A=1, B=0.1, Alpha=$User.ForwardDot), Use Linear Drag=true, Use Rotational Drag=false)
   &InheritSourceMovement(
    Applied Position Scale=&RandomRangeVector(Minimum=<0 0 0>, Maximum=<-0.1 -0.1 -0.2>)
    ApplyForce=false
    ApplyPosition=true
    ApplyPositionFalloff=false
   ) {enabled=false}
   &SolveForcesAndVelocity(Clamp Velocity=false, Limit Acceleration=false)
   &ScaleSpriteSizeBySpeed(
    Max Scale Factor=&RandomRangeVector2D(Minimum=<0.22 3>, Maximum=<6 28>)
    Min Scale Factor=<0.15 1>
    Velocity Threshold=3800
   )
  ]
  rendererClasses [&NiagaraSpriteRendererProperties]
  renderers [{rendererClass=&NiagaraSpriteRendererProperties}]
 }
 Condensed_Sparks {
  emitterUpdateScript [
   &EmitterState(Life Cycle Mode=::NewEnumerator1, Loop Behavior=::NewEnumerator1, Loop Duration=6.64, UseLoopDelay=false)
   &SpawnRate(
    SpawnRate=&FloatFromCurve(
     FloatCurve=DataInterface() {propertyValues {Curve {keys [
         {interpMode=RCIM_Cubic, tangentMode=RCTM_User, time=0.7148, value=40}
         {arriveTangent=-118.7953, interpMode=RCIM_Cubic, leaveTangent=-118.7953, tangentMode=RCTM_User, time=0.9617, value=-1.5686}
        ]}}}
     CurveIndex=$Emitter.NormalizedLoopAge
    )
    Use Spawn Probability=true
   ) {module=SpawnRate002}
   &SpawnRate(
    Spawn Probability=&FN_FrontEndFloatModifier(Gameplay=1, Front End=0)
    SpawnGroup=1
    SpawnRate=&FloatFromCurve(
     FloatCurve=DataInterface() {propertyValues {Curve {keys [
         {interpMode=RCIM_Cubic, tangentMode=RCTM_User, time=0, value=0}
         {interpMode=RCIM_Cubic, tangentMode=RCTM_User, time=0.0277, value=170}
         {arriveTangent=-260.7543, interpMode=RCIM_Cubic, leaveTangent=-260.7543, tangentMode=RCTM_User, time=0.7257, value=104.7049}
         {interpMode=RCIM_Cubic, tangentMode=RCTM_User, time=1, value=0}
        ]}}}
     CurveIndex=$Emitter.NormalizedLoopAge
    )
    Use Spawn Probability=true
   )
   &SpawnRate(
    SpawnGroup=2
    SpawnRate=&FloatFromCurve(
     FloatCurve=DataInterface() {propertyValues {Curve {keys [
         {interpMode=RCIM_Cubic, tangentMode=RCTM_User, time=0, value=0}
         {arriveTangent=23.3418, interpMode=RCIM_Cubic, leaveTangent=23.3418, tangentMode=RCTM_User, time=0.0277, value=110}
         {arriveTangent=-58.9764, interpMode=RCIM_Cubic, leaveTangent=-58.9764, tangentMode=RCTM_User, time=0.7791, value=97.7763}
         {interpMode=RCIM_Cubic, tangentMode=RCTM_User, time=1, value=0}
        ]}}}
     CurveIndex=$Emitter.NormalizedLoopAge
    )
    Use Spawn Probability=false
   ) {module=SpawnRate001}
   &SpawnRate(
    SpawnGroup=3
    SpawnRate=&FloatFromCurve(
     FloatCurve=DataInterface() {propertyValues {Curve {keys [
         {arriveTangent=4983.9238, interpMode=RCIM_Cubic, leaveTangent=4983.9238, tangentMode=RCTM_User, time=0, value=0}
         {interpMode=RCIM_Cubic, tangentMode=RCTM_User, time=0.0363, value=65}
         {arriveTangent=-75.7856, interpMode=RCIM_Cubic, leaveTangent=-75.7856, tangentMode=RCTM_User, time=0.8395, value=49.9412}
         {arriveTangent=-791.5607, interpMode=RCIM_Cubic, leaveTangent=-791.5607, tangentMode=RCTM_User, time=1, value=0}
        ]}}}
     CurveIndex=$Emitter.NormalizedLoopAge
    )
    Use Spawn Probability=false
   ) {module=SpawnRate003}
   &SpawnRate(
    SpawnGroup=4
    SpawnRate=&FloatFromCurve(
     FloatCurve=DataInterface() {propertyValues {Curve {keys [
         {interpMode=RCIM_Cubic, tangentMode=RCTM_User, time=0, value=0}
         {arriveTangentWeight=0.1761, interpMode=RCIM_Cubic, leaveTangentWeight=4.9585, tangentMode=RCTM_User, time=0.0277, value=75}
         {arriveTangentWeight=25.8057, interpMode=RCIM_Cubic, leaveTangentWeight=6.3125, tangentMode=RCTM_User, time=0.8089, value=75}
         <at 1 0>
        ]}}}
     CurveIndex=$Emitter.NormalizedLoopAge
    )
    Use Spawn Probability=false
   ) {module=SpawnRate004}
   &SpawnRate(
    SpawnGroup=5
    SpawnRate=&FloatFromCurve(
     FloatCurve=DataInterface() {propertyValues {Curve {keys [
         {interpMode=RCIM_Cubic, tangentMode=RCTM_User, time=0, value=0}
         {arriveTangentWeight=0.1761, interpMode=RCIM_Cubic, leaveTangentWeight=5.2259, tangentMode=RCTM_User, time=0.0277, value=70}
         {arriveTangentWeight=27.1972, interpMode=RCIM_Cubic, leaveTangentWeight=4.921, tangentMode=RCTM_User, time=0.851, value=70}
         <at 1 0>
        ]}}}
     CurveIndex=$Emitter.NormalizedLoopAge
    )
    Use Spawn Probability=false
   ) {module=SpawnRate005}
   &SpawnRate(
    SpawnGroup=6
    SpawnRate=&FloatFromCurve(
     FloatCurve=DataInterface() {propertyValues {Curve {keys [
         {interpMode=RCIM_Cubic, tangentMode=RCTM_User, time=0, value=0}
         {arriveTangentWeight=0.1761, interpMode=RCIM_Cubic, leaveTangentWeight=5.2259, tangentMode=RCTM_User, time=0.0309, value=300}
         {arriveTangentWeight=27.1972, interpMode=RCIM_Cubic, leaveTangentWeight=4.921, tangentMode=RCTM_User, time=0.8816, value=307.5162}
         <at 0.9755 0>
        ]}}}
     CurveIndex=$Emitter.NormalizedLoopAge
    )
    Use Spawn Probability=false
   ) {module=SpawnRate006}
   &SpawnRate(
    SpawnGroup=7
    SpawnRate=&FloatFromCurve(
     FloatCurve=DataInterface() {propertyValues {Curve {
        defaultValue=1
        keys [
         {arriveTangent=78.1641, interpMode=RCIM_Cubic, leaveTangent=78.1641, tangentMode=RCTM_User, time=0.1854, value=2.0588}
         {arriveTangent=187.1201, interpMode=RCIM_Cubic, leaveTangent=187.1201, tangentMode=RCTM_User, time=0.4492, value=46.3661}
         {arriveTangent=-181.6984, interpMode=RCIM_Cubic, leaveTangent=-181.6984, tangentMode=RCTM_User, time=0.7677, value=43.7717}
         {arriveTangent=-2.3561, interpMode=RCIM_Cubic, leaveTangent=-2.3561, tangentMode=RCTM_User, time=0.973, value=0}
        ]
       }}}
     CurveIndex=$Emitter.NormalizedLoopAge
    )
    Use Spawn Probability=false
   ) {module=SpawnRate007}
  ]
  particleSpawnScript [
   /SetVariables(Particles.Custom Random:NiagaraFloat=&RandomRangeFloat(Minimum=0.2, Maximum=0.8)) {isSetParameters=true, module=SetVariables_#4}
   /SetVariables(Particles.VisibilityTag=&SpawnGroup()) {isSetParameters=true, module=SetVariables_#5}
   &PartitionParticles()
   &InitializeParticle(
    Color Mode=::NewEnumerator0
    Lifetime=&MergeHelper_FloatSwitch(
     0=&RandomRangeFloat(Minimum=0.1, Maximum=0.13)
     1=&RandomRangeFloat(Minimum=0.13, Maximum=0.22)
     2=&RandomRangeFloat(Minimum=0.1, Maximum=0.2)
     3=&RandomRangeFloat(Minimum=0.2, Maximum=0.23)
     4=&RandomRangeFloat(Minimum=0.3, Maximum=0.25)
     5=&RandomRangeFloat(Minimum=0.12, Maximum=0.26)
     6=&RandomRangeFloat(Minimum=0.01, Maximum=0.05)
     7=&RandomRangeFloat(Minimum=0.8, Maximum=1.6)
     Index=&SpawnGroup()
     Num=::NewEnumerator6
    )
    Sprite Size=&MergeHelper_Vector2Switch(
     0=&Vector2DFromFloat(Value=&RandomRangeFloat(Minimum=6, Maximum=10))
     1=&RandomRangeVector2D(Minimum=<1 1>, Maximum=<3 6>)
     2=&RandomRangeVector2D(Minimum=<0.2 0.4>, Maximum=<0.4 0.6>)
     3=&RandomRangeVector2D(Minimum=<0.5 0.6>, Maximum=<0.8 1>)
     4=<0.1 0.3>
     5=&RandomRangeVector2D(Minimum=<3 2.5>, Maximum=<8 10>)
     6=&RandomRangeVector2D(Minimum=<0.1 0.3>, Maximum=<0.3 0.6>)
     7=&Vector2DFromFloat(Value=&RandomRangeFloat(Minimum=30, Maximum=115))
     Index=&SpawnGroup()
     Num=::NewEnumerator6
    )
    Sprite Size Mode=::NewEnumerator4
    UsePositionOffset=false
    Write Material Random=true
   )
   &SkeletalMeshLocation(
    Bone / Socket Index=$Particles.PartitionParticles.Partition
    MaskSpawnGroup=false
    Mesh Sampling Type=::NewEnumerator5
    Skeletal Mesh=$System.Skeletal Mesh
    Socket Sampling Mode=::NewEnumerator5
   )
   &ShapeLocation(
    MaskSpawnGroup=false
    Override Seed=false
    Sphere Radius=&MergeHelper_FloatSwitch(0=0, 1=0, 2=3, 3=3.75, 4=4.25, 5=3, 6=5, 7=5, Index=&SpawnGroup(), Num=::NewEnumerator6)
    UseOwnerScale=true
   )
   &CameraOffset(Camera Offset Amount=&MergeHelper_FloatSwitch(
     0=0
     1=0
     2=&RandomRangeFloat(Minimum=2, Maximum=12)
     3=&RandomRangeFloat(Minimum=2, Maximum=12)
     4=&RandomRangeFloat(Minimum=2, Maximum=12)
     5=&RandomRangeFloat(Minimum=2, Maximum=12)
     6=&RandomRangeFloat(Minimum=2, Maximum=12)
     7=0
     Index=&SpawnGroup()
     Num=::NewEnumerator6
    ))
   &VortexForce(
    Origin Pull Amount=-30
    UseOriginOffset=false
    Vortex Axis=<1 0 0>
    Vortex Force Amount=&MergeHelper_FloatSwitch(
     0=&RandomRangeFloat(Minimum=355, Maximum=370)
     1=&RandomRangeFloat(Minimum=355, Maximum=370)
     2=&RandomRangeFloat(Minimum=360, Maximum=380)
     3=&RandomRangeFloat(Minimum=360, Maximum=380)
     4=&RandomRangeFloat(Minimum=360, Maximum=380)
     5=&RandomRangeFloat(Minimum=360, Maximum=380)
     6=&RandomRangeFloat(Minimum=360, Maximum=380)
     7=&RandomRangeFloat(Minimum=75, Maximum=115)
     Index=&SpawnGroup()
     Num=::NewEnumerator6
    )
    Vortex Origin:NiagaraPosition=&AddVectorToPosition(Position=$Engine.Emitter.SimulationPosition, Vector=<0 0 130>)
   )
   &ApplyInitialForces()
  ]
  particleUpdateScript [
   &ParticleState()
   /SetVariables(
    Particles.Classic Flicker:NiagaraFloat=&Waveform("Waveform [1]"=::NewEnumerator0, "[1] Frequency"=75, "[1] Amplitude Min/Max"=<0 2>)
    Particles.Classic Flicker Size:NiagaraFloat=&Waveform("Waveform [1]"=::NewEnumerator0, "[1] Frequency"=75, "[1] Amplitude Min/Max"=<0.15 1>)
    Particles.Classic Flicker Value:NiagaraFloat=&Waveform("Waveform [1]"=::NewEnumerator0, "[1] Frequency"=75, "[1] Amplitude Min/Max"=<2 8>)
   ) {isSetParameters=true, module=SetVariables_#6}
   /SetVariables(
    Particles.Flicker:NiagaraFloat=&Waveform("Waveform [1]"=::NewEnumerator0, "[1] Frequency"=15, "[1] Amplitude Min/Max"=<0 1>)
    Particles.Flicker Size:NiagaraFloat=&Waveform("Waveform [1]"=::NewEnumerator0, "[1] Frequency"=15, "[1] Amplitude Min/Max"=<0.25 1.5>)
    Particles.Flicker Value:NiagaraFloat=&Waveform("Waveform [1]"=::NewEnumerator0, "[1] Frequency"=15, "[1] Amplitude Min/Max"=<2 8>)
   ) {isSetParameters=true, module=SetVariables_#7}
   &ScaleSpriteSize(
    Scale Factor=&MergeHelper_Vector2Switch(
     0=&Multiply_Vector2D(
      A=&RandomRangeVector2D(Minimum=<2 2>, Maximum=<3 3>)
      B=&Vector2DFromCurve(Vector2Curve=DataInterface() {propertyValues {XCurve {keys [<at 0 1>,<at 1 0.5>]}, YCurve {keys [<at 0 1>,<at 1 0.5>]}}}, ScaleCurve=<1 1>)
     )
     1=&Vector2DFromFloat(Value=&Lerp_Float(
       A=1
       B=$Particles.Classic Flicker Size
       Alpha=&FloatFromCurve(FloatCurve=DataInterface() {propertyValues {Curve {keys [<at 0 0>,<at 0.2485 0.4244>,<at 0.45 1>,<at 1 1>]}}})
      ))
     2=&Vector2DFromCurve(
      Vector2Curve=DataInterface() {propertyValues {
        XCurve {keys [
          {arriveTangent=43.4086, interpMode=RCIM_Cubic, leaveTangent=43.4086, tangentMode=RCTM_User, time=0, value=0}
          {interpMode=RCIM_Cubic, tangentMode=RCTM_User, time=0.052, value=0.985}
          {arriveTangent=-0.4356, interpMode=RCIM_Cubic, leaveTangent=-0.4356, tangentMode=RCTM_User, time=0.5928, value=0.8437}
          {arriveTangent=-2.8823, interpMode=RCIM_Cubic, leaveTangent=-2.8823, tangentMode=RCTM_User, time=1, value=0}
         ]}
        YCurve {
         defaultValue=1
         keys [
          {arriveTangent=43.4086, interpMode=RCIM_Cubic, leaveTangent=43.4086, tangentMode=RCTM_User, time=0, value=0}
          {arriveTangentWeight=0.0173, interpMode=RCIM_Cubic, leaveTangentWeight=0.316, tangentMode=RCTM_User, time=0.052, value=0.985}
          {
           arriveTangent=-0.468
           arriveTangentWeight=0.2924
           interpMode=RCIM_Cubic
           leaveTangent=-0.468
           leaveTangentWeight=0.0236
           tangentMode=RCTM_User
           time=0.6057
           value=0.8431
          }
          {arriveTangent=-6.8563, arriveTangentWeight=0.3333, interpMode=RCIM_Cubic, leaveTangent=-6.8563, tangentMode=RCTM_User, time=1, value=0}
         ]
        }
       }}
      ScaleCurve=&RandomRangeVector2D(Minimum=<1 1>, Maximum=<4 5>)
     )
     3=&Vector2DFromCurve(
      Vector2Curve=DataInterface() {propertyValues {
        XCurve {keys [
          {arriveTangent=8.5969, interpMode=RCIM_Cubic, leaveTangent=8.5969, tangentMode=RCTM_User, time=0, value=1}
          {interpMode=RCIM_Cubic, tangentMode=RCTM_User, time=0.3444, value=2}
          {arriveTangent=-4.2468, interpMode=RCIM_Cubic, leaveTangent=-4.2468, tangentMode=RCTM_User, time=1, value=0}
         ]}
        YCurve {keys [
          {arriveTangent=8.5969, interpMode=RCIM_Cubic, leaveTangent=8.5969, tangentMode=RCTM_User, time=0, value=1}
          {interpMode=RCIM_Cubic, tangentMode=RCTM_User, time=0.3444, value=2}
          {arriveTangent=-8.7532, interpMode=RCIM_Cubic, leaveTangent=-8.7532, tangentMode=RCTM_User, time=1, value=0}
         ]}
       }}
      ScaleCurve=&Vector2DFromFloat(Value=&RandomRangeFloat(Minimum=0.8, Maximum=1.5))
     )
     4=&Vector2DFromCurve(
      Vector2Curve=DataInterface() {propertyValues {
        XCurve {keys [
          {arriveTangent=42.9702, interpMode=RCIM_Cubic, leaveTangent=42.9702, tangentMode=RCTM_User, time=-0.0043, value=0}
          {interpMode=RCIM_Cubic, tangentMode=RCTM_User, time=0.0755, value=2}
          {interpMode=RCIM_Cubic, tangentMode=RCTM_User, time=1, value=0}
         ]}
        YCurve {keys [
          {arriveTangent=42.9702, interpMode=RCIM_Cubic, leaveTangent=42.9702, tangentMode=RCTM_User, time=-0.0043, value=0}
          {interpMode=RCIM_Cubic, tangentMode=RCTM_User, time=0.0755, value=2}
          {arriveTangent=-1.6294, interpMode=RCIM_Cubic, leaveTangent=-1.6294, tangentMode=RCTM_User, time=1, value=0}
         ]}
       }}
      ScaleCurve=&RandomRangeVector2D(Minimum=<0.75 0.75>, Maximum=<1 1>)
     )
     5=&Vector2DFromCurve(
      Vector2Curve=DataInterface() {propertyValues {
        XCurve {
         defaultValue=0
         keys [
          {arriveTangent=0.6985, interpMode=RCIM_Cubic, leaveTangent=0.6985, tangentMode=RCTM_User, time=0.425, value=0}
          {arriveTangent=1.747, arriveTangentWeight=0.1533, interpMode=RCIM_Cubic, leaveTangent=1.747, tangentMode=RCTM_User, time=0.9133, value=1}
          {
           arriveTangent=-19.7643
           arriveTangentWeight=0.0134
           interpMode=RCIM_Cubic
           leaveTangent=-19.7643
           tangentMode=RCTM_User
           time=0.9684
           value=0
          }
         ]
        }
        YCurve {
         defaultValue=0
         keys [
          {arriveTangent=0.6985, arriveTangentWeight=0.1694, interpMode=RCIM_Cubic, leaveTangent=0.6985, tangentMode=RCTM_User, time=0.425, value=0}
          {arriveTangent=1.747, arriveTangentWeight=0.1533, interpMode=RCIM_Cubic, leaveTangent=1.747, tangentMode=RCTM_User, time=0.9133, value=1}
          {
           arriveTangent=-19.7643
           arriveTangentWeight=0.0134
           interpMode=RCIM_Cubic
           leaveTangent=-19.7643
           tangentMode=RCTM_User
           time=0.9684
           value=0
          }
         ]
        }
       }}
      ScaleCurve=&RandomRangeVector2D(Minimum=<0.9 0.9>, Maximum=<1.25 1.25>)
     )
     6=&Vector2DFromCurve(
      Vector2Curve=DataInterface() {propertyValues {
        XCurve {
         defaultValue=1.306
         keys [
          {
           arriveTangent=0.2391
           interpMode=RCIM_Cubic
           leaveTangent=0.2391
           leaveTangentWeight=0.3327
           tangentMode=RCTM_User
           time=0.0018
           value=1.2938
          }
          {
           arriveTangent=-0.4123
           arriveTangentWeight=0.0661
           interpMode=RCIM_Cubic
           leaveTangent=-0.4123
           leaveTangentWeight=0.2666
           tangentMode=RCTM_User
           time=0.2001
           value=0.993
          }
          {arriveTangent=-2.7886, interpMode=RCIM_Cubic, leaveTangent=-2.7886, tangentMode=RCTM_User, time=1, value=0}
         ]
        }
        YCurve {keys [
          {arriveTangent=0.2391, interpMode=RCIM_Cubic, leaveTangent=0.2391, tangentMode=RCTM_User, time=0.0018, value=1.2938}
          {arriveTangent=-0.4123, interpMode=RCIM_Cubic, leaveTangent=-0.4123, tangentMode=RCTM_User, time=0.2001, value=0.993}
          {arriveTangent=-1.6274, interpMode=RCIM_Cubic, leaveTangent=-1.6274, tangentMode=RCTM_User, time=1, value=0}
         ]}
       }}
      ScaleCurve=&RandomRangeVector2D(Minimum=<3 3>, Maximum=<4 5>)
     )
     7=&Vector2DFromFloat(Value=&FloatFromCurve(FloatCurve=DataInterface() {propertyValues {Curve {
          defaultValue=0
          keys [
           {arriveTangent=3.2247, interpMode=RCIM_Cubic, leaveTangent=3.2247, tangentMode=RCTM_User, time=0, value=0.1}
           {arriveTangent=0.9, interpMode=RCIM_Cubic, leaveTangent=0.9, time=0.2526, value=0.6292}
           {arriveTangent=-0.0053, interpMode=RCIM_Cubic, leaveTangent=-0.0053, tangentMode=RCTM_User, time=1, value=1}
          ]
         }}}))
     Index=&SpawnGroup()
     Num=::NewEnumerator6
    )
    Scale Sprite Size Mode=::NewEnumerator2
   )
   &ScaleColor(
    Scale Alpha=&MergeHelper_FloatSwitch(
     0=0
     1=0
     2=&Multiply_Float(
      A=&FloatFromCurve(FloatCurve=DataInterface() {propertyValues {Curve {
          defaultValue=1
          keys [
           {arriveTangent=-0.1558, interpMode=RCIM_Cubic, leaveTangent=-0.1558, tangentMode=RCTM_User, time=-0.0073, value=1}
           {arriveTangent=-1.2145, interpMode=RCIM_Cubic, leaveTangent=-1.2145, tangentMode=RCTM_User, time=0.4515, value=0.2667}
           {arriveTangent=-0.1147, interpMode=RCIM_Cubic, leaveTangent=-0.1147, tangentMode=RCTM_User, time=1, value=0}
          ]
         }}})
      B=$Particles.Flicker
     )
     3=&Multiply_Float(
      A=&FloatFromCurve(FloatCurve=DataInterface() {propertyValues {Curve {
          defaultValue=1
          keys [
           {interpMode=RCIM_Cubic, tangentMode=RCTM_User, time=-0.0073, value=1}
           {arriveTangent=-1.2145, interpMode=RCIM_Cubic, leaveTangent=-1.2145, tangentMode=RCTM_User, time=0.563, value=0.7725}
           {arriveTangent=-2.6173, interpMode=RCIM_Cubic, leaveTangent=-2.6173, tangentMode=RCTM_User, time=1, value=0}
          ]
         }}})
      B=$Particles.Flicker
     )
     4=1
     5=&FloatFromCurve(FloatCurve=DataInterface() {propertyValues {Curve {
         defaultValue=1
         keys [
          {interpMode=RCIM_Cubic, tangentMode=RCTM_User, time=0.2664, value=0}
          {arriveTangent=-3.8042, interpMode=RCIM_Cubic, leaveTangent=-3.8042, tangentMode=RCTM_User, time=1, value=1}
         ]
        }}})
     6=&FloatFromCurve(FloatCurve=DataInterface() {propertyValues {Curve {
         defaultValue=1
         keys [
          {tangentMode=RCTM_User, time=0, value=1}
          {arriveTangent=-0.0405, interpMode=RCIM_Cubic, leaveTangent=-0.0405, tangentMode=RCTM_User, time=0.7009, value=1}
          {arriveTangent=-8.0089, interpMode=RCIM_Cubic, leaveTangent=-8.0089, tangentMode=RCTM_User, time=1, value=0}
         ]
        }}})
     7=0
     Index=&SpawnGroup()
     Num=::NewEnumerator6
    )
    Scale RGB=&MergeHelper_VectorSwitch(
     0=<0 0 0>
     1=<0 0 0>
     2=&MakeVectorFromLinearColorRGB(Color=&ColorFromCurve(ColorCurve=DataInterface() {propertyValues {
         AlphaCurve {keys [<at 0.2899 1 RCIM_Cubic>]}
         BlueCurve {keys [
           <at 0 9.9666>
           {arriveTangent=-43.7312, interpMode=RCIM_Cubic, leaveTangent=-43.7312, time=0.0532, value=0.2992}
           <at 0.2279 0>
           <at 0.5806 0>
           <at 0.9904 0>
          ]}
         GreenCurve {keys [
           <at 0 10>
           {arriveTangent=-42.1571, interpMode=RCIM_Cubic, leaveTangent=-42.1571, time=0.0532, value=2.4531}
           <at 0.2279 0.3922>
           <at 0.5806 0.0737>
           <at 0.9904 0.4318>
          ]}
         RedCurve {keys [
           <at 0 4.9499>
           {arriveTangent=-17.3312, interpMode=RCIM_Cubic, leaveTangent=-17.3312, time=0.0532, value=5}
           <at 0.2279 1>
           <at 0.5806 1>
           <at 0.9904 12>
          ]}
        }}))
     3=&MakeVectorFromLinearColorRGB(Color=<rgba 8 1.8933 0.4786 1>)
     4=&MakeVectorFromLinearColorRGB(Color=&ColorFromCurve(ColorCurve=DataInterface() {propertyValues {
         AlphaCurve {keys [<at 0.2041 1 RCIM_Cubic>]}
         BlueCurve {keys [
           <at 0 4.9833>
           {arriveTangent=-38.7566, interpMode=RCIM_Cubic, leaveTangent=-38.7566, time=0.025, value=0.7359}
           <at 0.1286 0>
           <at 0.3687 0>
           <at 0.9084 0.0275>
           <at 0.9981 0>
          ]}
         GreenCurve {keys [
           <at 0 5>
           {arriveTangent=-36.9042, interpMode=RCIM_Cubic, leaveTangent=-36.9042, time=0.025, value=3.2282}
           <at 0.1286 0.2549>
           <at 0.3687 0.1473>
           <at 0.9084 0.0234>
           <at 0.9981 0.4318>
          ]}
         RedCurve {keys [
           <at 0 2.4749>
           {arriveTangent=-11.471, interpMode=RCIM_Cubic, leaveTangent=-11.471, time=0.025, value=5}
           <at 0.1286 1>
           <at 0.3687 1>
           <at 0.9084 0.5781>
           <at 0.9981 12>
          ]}
        }}))
     5=&MakeVectorFromLinearColorRGB(Color=&RandomRangeLinearColor(Minimum=<rgba 10 0.5374 0 1>, Maximum=<rgba 10 6.9325 0 1>))
     6=&MakeVectorFromLinearColorRGB(Color=<rgba 20 14.5374 8.5091 1>)
     7=<0 0 0>
     Index=&SpawnGroup()
     Num=::NewEnumerator6
    )
    ScaleA=true
    ScaleRGB=true
   )
   /SetVariables(Particles.0 - Base Streaks Color:LinearColor=&ColorFromCurve(ColorCurve=DataInterface() {propertyValues {
       AlphaCurve {keys [<at 0 0.7>,<at 0.6507 0.18>,<at 1 0>]}
       BlueCurve {keys [<at 0.0424 2.6875>,<at 0.3289 0>,<at 0.6955 0>,<at 1 0>]}
       GreenCurve {keys [<at 0.0424 2.919>,<at 0.3289 1.3329>,<at 0.6955 0.3963>,<at 1 0.0724>]}
       RedCurve {keys [<at 0.0424 3>,<at 0.3289 2.5942>,<at 0.6955 1.8651>,<at 1 1>]}
      }})) {isSetParameters=true, module=SetVariables_#8}
   /SetVariables(Particles.1 - Classic Sparks Color:LinearColor=/Niagara/DynamicInputs/LinearColor/MakeLinearColorFromVector4(Vector:Vector4f=&Multiply_Vector4(
      A=&MakeVector4_FromLinearColor(LinearColor=&ColorFromCurve(ColorCurve=DataInterface() {propertyValues {
          AlphaCurve {keys [<at 0.122 1>]}
          BlueCurve {keys [<at 0 3.0372>,<at 0.1056 0>,<at 0.2341 0>,<at 0.8178 0>,<at 0.8544 0>,<at 0.9314 0>,<at 0.9928 0>]}
          GreenCurve {keys [<at 0 7.61>,<at 0.1056 3.3715>,<at 0.2341 0.7094>,<at 0.8178 0.1839>,<at 0.8544 5.7311>,<at 0.9314 0.6262>,<at 0.9928 0.2902>]}
          RedCurve {keys [<at 0 10>,<at 0.1056 10>,<at 0.2341 7.5>,<at 0.8178 7.5>,<at 0.8544 8.0224>,<at 0.9314 10>,<at 0.9928 10>]}
         }}))
      B=&MakeVector4_FromLinearColor(LinearColor=&MakeLinearColorFromVectorAndFloat(
        "Vector (RGB)"=&Add_Vector(
         A=<1 1 1>
         B=&MakeVectorFromLinearColorHSV(Color:LinearColor=&MakeLinearColorFromHSV(
           HSV From:NiagaraInt32=1
           HSV:Vector3f=?
           Hue:NiagaraFloat=0
           Saturation:NiagaraFloat=0
           Value:NiagaraFloat=&Lerp_Float(
            A=0
            B=$Particles.Classic Flicker Value
            Alpha=&FloatFromCurve(FloatCurve=DataInterface() {propertyValues {Curve {keys [<at 0 0>,<at 0.5 0>,<at 0.55 0.9624>,<at 1 1>]}}})
           )
           Alpha:NiagaraFloat=1
          ))
        )
        "Float (Alpha)"=&Lerp_Float(
         A=1
         B=$Particles.Classic Flicker
         Alpha=&FloatFromCurve(FloatCurve=DataInterface() {propertyValues {Curve {keys [<at 0 0>,<at 0.25 0>,<at 0.35 1>,<at 1 1>]}}})
        )
       ))
     ))) {isSetParameters=true, module=SetVariables_#9}
   /SetVariables(Particles.7 - Smoke Color:LinearColor=&ColorFromCurve(ColorCurve=DataInterface() {propertyValues {
       AlphaCurve {keys [<at 0 0>,<at 0.4023 0.05>,<at 0.5672 0.1>,<at 0.7418 0.05>,<at 0.9752 0>]}
       BlueCurve {keys [<at 0 0.083>,<at 0.3354 0.0297>,<at 0.6977 0.0147>,<at 0.9899 0.0033>]}
       GreenCurve {keys [<at 0 0.0964>,<at 0.3354 0.035>,<at 0.6977 0.0053>,<at 0.9899 0.0003>]}
       RedCurve {keys [<at 0 0.2656>,<at 0.3354 0.0781>,<at 0.6977 0.0208>,<at 0.9899 0.0052>]}
      }})) {isSetParameters=true, module=SetVariables_#10}
   &BlendSpace("Position Blend (World <-> Local)":NiagaraFloat=&MakeCustomFloatFromBool(
     Bool=&SetBoolByFloatComparison(A=&VectorLength(Vector=$Engine.Owner.Velocity), B=301)
     True Float=1
     False Float=&MaskFloat(
      Mask Source=0
      Mask Parameter=?
      Mask=7
      Masked Float=&FloatFromCurve(FloatCurve=DataInterface() {propertyValues {Curve {
          defaultValue=0
          keys [{interpMode=RCIM_Cubic, tangentMode=RCTM_User, time=0, value=1}, {interpMode=RCIM_Cubic, tangentMode=RCTM_User, time=0.63, value=0}]
         }}})
      Passthrough Float=&FloatFromCurve(FloatCurve=DataInterface() {propertyValues {Curve {
          defaultValue=0
          keys [{interpMode=RCIM_Cubic, tangentMode=RCTM_User, time=0, value=1}, {interpMode=RCIM_Cubic, tangentMode=RCTM_User, time=0.7, value=0.2}]
         }}})
     )
     Translation Contribution:NiagaraFloat=1
     Rotation Contribution:NiagaraFloat=1
    )) {module=BlendSpace001}
   &CurlNoiseForce(
    Bias Noise=false
    Noise Frequency=&MaskFloat(
     Mask Source=0
     Mask Parameter=?
     Mask=7
     Masked Float=&RandomRangeFloat(Minimum=20, Maximum=120)
     Passthrough Float=0
    )
    Noise Strength=&MaskFloat(
     Mask Source=0
     Mask Parameter=?
     Mask=7
     Masked Float=&RandomRangeFloat(Minimum=300, Maximum=900)
     Passthrough Float=0
    )
    Pan Noise=false
   )
   &InheritSourceMovement(
    Applied Position Scale=&MergeHelper_VectorSwitch(
     0=<0 0 0>
     1=<0.15 0.15 0.15>
     2=<0 0 0>
     3=<0 0 0>
     4=<0 0 0>
     5=<0 0 0>
     6=<0 0 0>
     7=<0 0 0>
     Index=&SpawnGroup()
     Num=::NewEnumerator6
    )
    ApplyForce=false
    ApplyPosition=true
    ApplyPositionFalloff=false
   ) {module=InheritSourceMovement001}
   &Drag(
    Drag=&MergeHelper_FloatSwitch(
     0=0
     1=0
     2=0
     3=0
     4=0
     5=0
     6=0
     7=&RandomRangeFloat(Minimum=1, Maximum=2.5)
     Index=&SpawnGroup()
     Num=::NewEnumerator6
    )
    Use Linear Drag=true
    Use Rotational Drag=false
   )
   &SubUVAnimation(
    Output Alternate Renderer Binding=true
    Play Rate:NiagaraFloat=&RandomRangeFloat(Minimum=-0.25, Maximum=-0.5)
    Playback Scale Curve=DataInterface() {propertyValues {Curve {keys [
        {interpMode=RCIM_Cubic, tangentMode=RCTM_User, time=0.25, value=1}
        {arriveTangent=-1.2656, interpMode=RCIM_Cubic, leaveTangent=-1.2656, tangentMode=RCTM_User, time=1, value=0}
       ]}}}
    Random Start Frame=true
    "Sprite Renderer "=DataInterface() {propertyValues {SpriteRenderer=&self:NiagaraEmitter_1.NiagaraSpriteRendererProperties_7}}
    Start Frame Offset:NiagaraInt32=&RandomRangeInt(Minimum=0, Maximum=63)
    SubUV Animation Mode:ENiagara_SubUVLookupModeV2=::NewEnumerator3
    UseEndFrame=true
    UsePlayRateScale=false
    UseStartFrame=true
   )
   &SolveForcesAndVelocity(Clamp Velocity=false, Limit Acceleration=false)
   &ScaleSpriteSizeBySpeed(
    Max Scale Factor=&MergeHelper_Vector2Switch(
     0=<1.15 6>
     1=&RandomRangeVector2D(Minimum=<0.22 3>, Maximum=<6 28>)
     2=&RandomRangeVector2D(Minimum=<1 7>, Maximum=<4 15>)
     3=&RandomRangeVector2D(Minimum=<1 5>, Maximum=<5 9>)
     4=&RandomRangeVector2D(Minimum=<1 6>, Maximum=<5 10>)
     5=<1 1>
     6=<1 5>
     7=<1 1>
     Index=&SpawnGroup()
     Num=::NewEnumerator6
    )
    Min Scale Factor=&MergeHelper_Vector2Switch(
     0=<0.75 1>
     1=<0.15 1>
     2=<0.5 2>
     3=<0.5 2>
     4=<0.5 2>
     5=<0.5 2>
     6=<0.5 2>
     7=<1 1>
     Index=&SpawnGroup()
     Num=::NewEnumerator6
    )
    Velocity Threshold:NiagaraFloat=&MergeHelper_FloatSwitch(0=1000, 1=3800, 2=200, 3=200, 4=200, 5=200, 6=200, 7=0, Index=&SpawnGroup(), Num=::NewEnumerator6)
   )
  ]
  rendererClasses [&NiagaraSpriteRendererProperties]
  renderers [
   {rendererClass=&NiagaraSpriteRendererProperties}
   {rendererClass=&NiagaraSpriteRendererProperties, rendererIndex=1}
   {rendererClass=&NiagaraSpriteRendererProperties, rendererIndex=2}
   {rendererClass=&NiagaraSpriteRendererProperties, rendererIndex=3}
   {rendererClass=&NiagaraSpriteRendererProperties, rendererIndex=4}
   {rendererClass=&NiagaraSpriteRendererProperties, rendererIndex=5}
   {rendererClass=&NiagaraSpriteRendererProperties, rendererIndex=6}
   {rendererClass=&NiagaraSpriteRendererProperties, rendererIndex=7}
  ]
 }
 Condensed_Sparks001 {
  simTarget=GPUComputeSim
  emitterUpdateScript [
   &EmitterState(Life Cycle Mode=::NewEnumerator1, Loop Behavior=::NewEnumerator1, Loop Duration=6.64, UseLoopDelay=false)
   &SpawnRate(
    SpawnRate=&FloatFromCurve(
     FloatCurve=DataInterface() {propertyValues {Curve {keys [
         {interpMode=RCIM_Cubic, tangentMode=RCTM_User, time=0.7148, value=40}
         {arriveTangent=-118.7953, interpMode=RCIM_Cubic, leaveTangent=-118.7953, tangentMode=RCTM_User, time=0.9617, value=-1.5686}
        ]}}}
     CurveIndex=$Emitter.NormalizedLoopAge
    )
    Use Spawn Probability=true
   ) {module=SpawnRate002}
   &SpawnRate(
    Spawn Probability=&FN_FrontEndFloatModifier(Gameplay=1, Front End=0)
    SpawnGroup=1
    SpawnRate=&FloatFromCurve(
     FloatCurve=DataInterface() {propertyValues {Curve {keys [
         {interpMode=RCIM_Cubic, tangentMode=RCTM_User, time=0, value=0}
         {interpMode=RCIM_Cubic, tangentMode=RCTM_User, time=0.0277, value=170}
         {arriveTangent=-260.7543, interpMode=RCIM_Cubic, leaveTangent=-260.7543, tangentMode=RCTM_User, time=0.7257, value=104.7049}
         {interpMode=RCIM_Cubic, tangentMode=RCTM_User, time=1, value=0}
        ]}}}
     CurveIndex=$Emitter.NormalizedLoopAge
    )
    Use Spawn Probability=true
   )
   &SpawnRate(
    SpawnGroup=2
    SpawnRate=&FloatFromCurve(
     FloatCurve=DataInterface() {propertyValues {Curve {keys [
         {interpMode=RCIM_Cubic, tangentMode=RCTM_User, time=0, value=0}
         {arriveTangent=23.3418, interpMode=RCIM_Cubic, leaveTangent=23.3418, tangentMode=RCTM_User, time=0.0277, value=110}
         {arriveTangent=-58.9764, interpMode=RCIM_Cubic, leaveTangent=-58.9764, tangentMode=RCTM_User, time=0.7791, value=97.7763}
         {interpMode=RCIM_Cubic, tangentMode=RCTM_User, time=1, value=0}
        ]}}}
     CurveIndex=$Emitter.NormalizedLoopAge
    )
    Use Spawn Probability=false
   ) {module=SpawnRate001}
   &SpawnRate(
    SpawnGroup=3
    SpawnRate=&FloatFromCurve(
     FloatCurve=DataInterface() {propertyValues {Curve {keys [
         {arriveTangent=4983.9238, interpMode=RCIM_Cubic, leaveTangent=4983.9238, tangentMode=RCTM_User, time=0, value=0}
         {interpMode=RCIM_Cubic, tangentMode=RCTM_User, time=0.0363, value=65}
         {arriveTangent=-75.7856, interpMode=RCIM_Cubic, leaveTangent=-75.7856, tangentMode=RCTM_User, time=0.8395, value=49.9412}
         {arriveTangent=-791.5607, interpMode=RCIM_Cubic, leaveTangent=-791.5607, tangentMode=RCTM_User, time=1, value=0}
        ]}}}
     CurveIndex=$Emitter.NormalizedLoopAge
    )
    Use Spawn Probability=false
   ) {module=SpawnRate003}
   &SpawnRate(
    SpawnGroup=4
    SpawnRate=&FloatFromCurve(
     FloatCurve=DataInterface() {propertyValues {Curve {keys [
         {interpMode=RCIM_Cubic, tangentMode=RCTM_User, time=0, value=0}
         {arriveTangentWeight=0.1761, interpMode=RCIM_Cubic, leaveTangentWeight=4.9585, tangentMode=RCTM_User, time=0.0277, value=75}
         {arriveTangentWeight=25.8057, interpMode=RCIM_Cubic, leaveTangentWeight=6.3125, tangentMode=RCTM_User, time=0.8089, value=75}
         <at 1 0>
        ]}}}
     CurveIndex=$Emitter.NormalizedLoopAge
    )
    Use Spawn Probability=false
   ) {module=SpawnRate004}
   &SpawnRate(
    SpawnGroup=5
    SpawnRate=&FloatFromCurve(
     FloatCurve=DataInterface() {propertyValues {Curve {keys [
         {interpMode=RCIM_Cubic, tangentMode=RCTM_User, time=0, value=0}
         {arriveTangentWeight=0.1761, interpMode=RCIM_Cubic, leaveTangentWeight=5.2259, tangentMode=RCTM_User, time=0.0277, value=70}
         {arriveTangentWeight=27.1972, interpMode=RCIM_Cubic, leaveTangentWeight=4.921, tangentMode=RCTM_User, time=0.851, value=70}
         <at 1 0>
        ]}}}
     CurveIndex=$Emitter.NormalizedLoopAge
    )
    Use Spawn Probability=false
   ) {module=SpawnRate005}
   &SpawnRate(
    SpawnGroup=6
    SpawnRate=&FloatFromCurve(
     FloatCurve=DataInterface() {propertyValues {Curve {keys [
         {interpMode=RCIM_Cubic, tangentMode=RCTM_User, time=0, value=0}
         {arriveTangentWeight=0.1761, interpMode=RCIM_Cubic, leaveTangentWeight=5.2259, tangentMode=RCTM_User, time=0.0309, value=300}
         {arriveTangentWeight=27.1972, interpMode=RCIM_Cubic, leaveTangentWeight=4.921, tangentMode=RCTM_User, time=0.8816, value=307.5162}
         <at 0.9755 0>
        ]}}}
     CurveIndex=$Emitter.NormalizedLoopAge
    )
    Use Spawn Probability=false
   ) {module=SpawnRate006}
   &SpawnRate(
    SpawnGroup=7
    SpawnRate=&FloatFromCurve(
     FloatCurve=DataInterface() {propertyValues {Curve {
        defaultValue=1
        keys [
         {arriveTangent=78.1641, interpMode=RCIM_Cubic, leaveTangent=78.1641, tangentMode=RCTM_User, time=0.1854, value=2.0588}
         {arriveTangent=187.1201, interpMode=RCIM_Cubic, leaveTangent=187.1201, tangentMode=RCTM_User, time=0.4492, value=46.3661}
         {arriveTangent=-181.6984, interpMode=RCIM_Cubic, leaveTangent=-181.6984, tangentMode=RCTM_User, time=0.7677, value=43.7717}
         {arriveTangent=-2.3561, interpMode=RCIM_Cubic, leaveTangent=-2.3561, tangentMode=RCTM_User, time=0.973, value=0}
        ]
       }}}
     CurveIndex=$Emitter.NormalizedLoopAge
    )
    Use Spawn Probability=false
   ) {module=SpawnRate007}
  ]
  particleSpawnScript [
   /SetVariables(Particles.Custom Random:NiagaraFloat=&RandomRangeFloat(Minimum=0.2, Maximum=0.8)) {isSetParameters=true, module=SetVariables_#4}
   /SetVariables(Particles.VisibilityTag=&SpawnGroup()) {isSetParameters=true, module=SetVariables_#5}
   &PartitionParticles()
   &InitializeParticle(
    Color Mode=::NewEnumerator0
    Lifetime=&MergeHelper_FloatSwitch(
     0=&RandomRangeFloat(Minimum=0.1, Maximum=0.13)
     1=&RandomRangeFloat(Minimum=0.13, Maximum=0.22)
     2=&RandomRangeFloat(Minimum=0.1, Maximum=0.2)
     3=&RandomRangeFloat(Minimum=0.2, Maximum=0.23)
     4=&RandomRangeFloat(Minimum=0.3, Maximum=0.25)
     5=&RandomRangeFloat(Minimum=0.12, Maximum=0.26)
     6=&RandomRangeFloat(Minimum=0.01, Maximum=0.05)
     7=&RandomRangeFloat(Minimum=0.8, Maximum=1.6)
     Index=&SpawnGroup()
     Num=::NewEnumerator6
    )
    Sprite Size=&MergeHelper_Vector2Switch(
     0=&Vector2DFromFloat(Value=&RandomRangeFloat(Minimum=6, Maximum=10))
     1=&RandomRangeVector2D(Minimum=<1 1>, Maximum=<3 6>)
     2=&RandomRangeVector2D(Minimum=<0.2 0.4>, Maximum=<0.4 0.6>)
     3=&RandomRangeVector2D(Minimum=<0.5 0.6>, Maximum=<0.8 1>)
     4=<0.1 0.3>
     5=&RandomRangeVector2D(Minimum=<3 2.5>, Maximum=<8 10>)
     6=&RandomRangeVector2D(Minimum=<0.1 0.3>, Maximum=<0.3 0.6>)
     7=&Vector2DFromFloat(Value=&RandomRangeFloat(Minimum=30, Maximum=115))
     Index=&SpawnGroup()
     Num=::NewEnumerator6
    )
    Sprite Size Mode=::NewEnumerator4
    UsePositionOffset=false
    Write Material Random=true
   )
   &SkeletalMeshLocation(
    Bone / Socket Index=$Particles.PartitionParticles.Partition
    MaskSpawnGroup=false
    Mesh Sampling Type=::NewEnumerator5
    Skeletal Mesh=$System.Skeletal Mesh
    Socket Sampling Mode=::NewEnumerator5
   )
   &ShapeLocation(
    MaskSpawnGroup=false
    Override Seed=false
    Sphere Radius=&MergeHelper_FloatSwitch(0=0, 1=0, 2=3, 3=3.75, 4=4.25, 5=3, 6=5, 7=5, Index=&SpawnGroup(), Num=::NewEnumerator6)
    UseOwnerScale=true
   )
   &CameraOffset(Camera Offset Amount=&MergeHelper_FloatSwitch(
     0=0
     1=0
     2=&RandomRangeFloat(Minimum=2, Maximum=12)
     3=&RandomRangeFloat(Minimum=2, Maximum=12)
     4=&RandomRangeFloat(Minimum=2, Maximum=12)
     5=&RandomRangeFloat(Minimum=2, Maximum=12)
     6=&RandomRangeFloat(Minimum=2, Maximum=12)
     7=0
     Index=&SpawnGroup()
     Num=::NewEnumerator6
    ))
   &VortexForce(
    Origin Pull Amount=-30
    UseOriginOffset=false
    Vortex Axis=<1 0 0>
    Vortex Force Amount=&MergeHelper_FloatSwitch(
     0=&RandomRangeFloat(Minimum=355, Maximum=370)
     1=&RandomRangeFloat(Minimum=355, Maximum=370)
     2=&RandomRangeFloat(Minimum=360, Maximum=380)
     3=&RandomRangeFloat(Minimum=360, Maximum=380)
     4=&RandomRangeFloat(Minimum=360, Maximum=380)
     5=&RandomRangeFloat(Minimum=360, Maximum=380)
     6=&RandomRangeFloat(Minimum=360, Maximum=380)
     7=&RandomRangeFloat(Minimum=75, Maximum=115)
     Index=&SpawnGroup()
     Num=::NewEnumerator6
    )
    Vortex Origin:NiagaraPosition=&AddVectorToPosition(Position=$Engine.Emitter.SimulationPosition, Vector=<0 0 130>)
   )
   &ApplyInitialForces()
  ]
  particleUpdateScript [
   &ParticleState()
   /SetVariables(
    Particles.Classic Flicker:NiagaraFloat=&Waveform("Waveform [1]"=::NewEnumerator0, "[1] Frequency"=75, "[1] Amplitude Min/Max"=<0 2>)
    Particles.Classic Flicker Size:NiagaraFloat=&Waveform("Waveform [1]"=::NewEnumerator0, "[1] Frequency"=75, "[1] Amplitude Min/Max"=<0.15 1>)
    Particles.Classic Flicker Value:NiagaraFloat=&Waveform("Waveform [1]"=::NewEnumerator0, "[1] Frequency"=75, "[1] Amplitude Min/Max"=<2 8>)
   ) {isSetParameters=true, module=SetVariables_#6}
   /SetVariables(
    Particles.Flicker:NiagaraFloat=&Waveform("Waveform [1]"=::NewEnumerator0, "[1] Frequency"=15, "[1] Amplitude Min/Max"=<0 1>)
    Particles.Flicker Size:NiagaraFloat=&Waveform("Waveform [1]"=::NewEnumerator0, "[1] Frequency"=15, "[1] Amplitude Min/Max"=<0.25 1.5>)
    Particles.Flicker Value:NiagaraFloat=&Waveform("Waveform [1]"=::NewEnumerator0, "[1] Frequency"=15, "[1] Amplitude Min/Max"=<2 8>)
   ) {isSetParameters=true, module=SetVariables_#7}
   &ScaleSpriteSize(
    Scale Factor=&MergeHelper_Vector2Switch(
     0=&Multiply_Vector2D(
      A=&RandomRangeVector2D(Minimum=<2 2>, Maximum=<3 3>)
      B=&Vector2DFromCurve(Vector2Curve=DataInterface() {propertyValues {XCurve {keys [<at 0 1>,<at 1 0.5>]}, YCurve {keys [<at 0 1>,<at 1 0.5>]}}}, ScaleCurve=<1 1>)
     )
     1=&Vector2DFromFloat(Value=&Lerp_Float(
       A=1
       B=$Particles.Classic Flicker Size
       Alpha=&FloatFromCurve(FloatCurve=DataInterface() {propertyValues {Curve {keys [<at 0 0>,<at 0.2485 0.4244>,<at 0.45 1>,<at 1 1>]}}})
      ))
     2=&Vector2DFromCurve(
      Vector2Curve=DataInterface() {propertyValues {
        XCurve {keys [
          {arriveTangent=43.4086, interpMode=RCIM_Cubic, leaveTangent=43.4086, tangentMode=RCTM_User, time=0, value=0}
          {interpMode=RCIM_Cubic, tangentMode=RCTM_User, time=0.052, value=0.985}
          {arriveTangent=-0.4356, interpMode=RCIM_Cubic, leaveTangent=-0.4356, tangentMode=RCTM_User, time=0.5928, value=0.8437}
          {arriveTangent=-2.8823, interpMode=RCIM_Cubic, leaveTangent=-2.8823, tangentMode=RCTM_User, time=1, value=0}
         ]}
        YCurve {
         defaultValue=1
         keys [
          {arriveTangent=43.4086, interpMode=RCIM_Cubic, leaveTangent=43.4086, tangentMode=RCTM_User, time=0, value=0}
          {arriveTangentWeight=0.0173, interpMode=RCIM_Cubic, leaveTangentWeight=0.316, tangentMode=RCTM_User, time=0.052, value=0.985}
          {
           arriveTangent=-0.468
           arriveTangentWeight=0.2924
           interpMode=RCIM_Cubic
           leaveTangent=-0.468
           leaveTangentWeight=0.0236
           tangentMode=RCTM_User
           time=0.6057
           value=0.8431
          }
          {arriveTangent=-6.8563, arriveTangentWeight=0.3333, interpMode=RCIM_Cubic, leaveTangent=-6.8563, tangentMode=RCTM_User, time=1, value=0}
         ]
        }
       }}
      ScaleCurve=&RandomRangeVector2D(Minimum=<1 1>, Maximum=<4 5>)
     )
     3=&Vector2DFromCurve(
      Vector2Curve=DataInterface() {propertyValues {
        XCurve {keys [
          {arriveTangent=8.5969, interpMode=RCIM_Cubic, leaveTangent=8.5969, tangentMode=RCTM_User, time=0, value=1}
          {interpMode=RCIM_Cubic, tangentMode=RCTM_User, time=0.3444, value=2}
          {arriveTangent=-4.2468, interpMode=RCIM_Cubic, leaveTangent=-4.2468, tangentMode=RCTM_User, time=1, value=0}
         ]}
        YCurve {keys [
          {arriveTangent=8.5969, interpMode=RCIM_Cubic, leaveTangent=8.5969, tangentMode=RCTM_User, time=0, value=1}
          {interpMode=RCIM_Cubic, tangentMode=RCTM_User, time=0.3444, value=2}
          {arriveTangent=-8.7532, interpMode=RCIM_Cubic, leaveTangent=-8.7532, tangentMode=RCTM_User, time=1, value=0}
         ]}
       }}
      ScaleCurve=&Vector2DFromFloat(Value=&RandomRangeFloat(Minimum=0.8, Maximum=1.5))
     )
     4=&Vector2DFromCurve(
      Vector2Curve=DataInterface() {propertyValues {
        XCurve {keys [
          {arriveTangent=42.9702, interpMode=RCIM_Cubic, leaveTangent=42.9702, tangentMode=RCTM_User, time=-0.0043, value=0}
          {interpMode=RCIM_Cubic, tangentMode=RCTM_User, time=0.0755, value=2}
          {interpMode=RCIM_Cubic, tangentMode=RCTM_User, time=1, value=0}
         ]}
        YCurve {keys [
          {arriveTangent=42.9702, interpMode=RCIM_Cubic, leaveTangent=42.9702, tangentMode=RCTM_User, time=-0.0043, value=0}
          {interpMode=RCIM_Cubic, tangentMode=RCTM_User, time=0.0755, value=2}
          {arriveTangent=-1.6294, interpMode=RCIM_Cubic, leaveTangent=-1.6294, tangentMode=RCTM_User, time=1, value=0}
         ]}
       }}
      ScaleCurve=&RandomRangeVector2D(Minimum=<0.75 0.75>, Maximum=<1 1>)
     )
     5=&Vector2DFromCurve(
      Vector2Curve=DataInterface() {propertyValues {
        XCurve {
         defaultValue=0
         keys [
          {arriveTangent=0.6985, interpMode=RCIM_Cubic, leaveTangent=0.6985, tangentMode=RCTM_User, time=0.425, value=0}
          {arriveTangent=1.747, arriveTangentWeight=0.1533, interpMode=RCIM_Cubic, leaveTangent=1.747, tangentMode=RCTM_User, time=0.9133, value=1}
          {
           arriveTangent=-19.7643
           arriveTangentWeight=0.0134
           interpMode=RCIM_Cubic
           leaveTangent=-19.7643
           tangentMode=RCTM_User
           time=0.9684
           value=0
          }
         ]
        }
        YCurve {
         defaultValue=0
         keys [
          {arriveTangent=0.6985, arriveTangentWeight=0.1694, interpMode=RCIM_Cubic, leaveTangent=0.6985, tangentMode=RCTM_User, time=0.425, value=0}
          {arriveTangent=1.747, arriveTangentWeight=0.1533, interpMode=RCIM_Cubic, leaveTangent=1.747, tangentMode=RCTM_User, time=0.9133, value=1}
          {
           arriveTangent=-19.7643
           arriveTangentWeight=0.0134
           interpMode=RCIM_Cubic
           leaveTangent=-19.7643
           tangentMode=RCTM_User
           time=0.9684
           value=0
          }
         ]
        }
       }}
      ScaleCurve=&RandomRangeVector2D(Minimum=<0.9 0.9>, Maximum=<1.25 1.25>)
     )
     6=&Vector2DFromCurve(
      Vector2Curve=DataInterface() {propertyValues {
        XCurve {
         defaultValue=1.306
         keys [
          {
           arriveTangent=0.2391
           interpMode=RCIM_Cubic
           leaveTangent=0.2391
           leaveTangentWeight=0.3327
           tangentMode=RCTM_User
           time=0.0018
           value=1.2938
          }
          {
           arriveTangent=-0.4123
           arriveTangentWeight=0.0661
           interpMode=RCIM_Cubic
           leaveTangent=-0.4123
           leaveTangentWeight=0.2666
           tangentMode=RCTM_User
           time=0.2001
           value=0.993
          }
          {arriveTangent=-2.7886, interpMode=RCIM_Cubic, leaveTangent=-2.7886, tangentMode=RCTM_User, time=1, value=0}
         ]
        }
        YCurve {keys [
          {arriveTangent=0.2391, interpMode=RCIM_Cubic, leaveTangent=0.2391, tangentMode=RCTM_User, time=0.0018, value=1.2938}
          {arriveTangent=-0.4123, interpMode=RCIM_Cubic, leaveTangent=-0.4123, tangentMode=RCTM_User, time=0.2001, value=0.993}
          {arriveTangent=-1.6274, interpMode=RCIM_Cubic, leaveTangent=-1.6274, tangentMode=RCTM_User, time=1, value=0}
         ]}
       }}
      ScaleCurve=&RandomRangeVector2D(Minimum=<3 3>, Maximum=<4 5>)
     )
     7=&Vector2DFromFloat(Value=&FloatFromCurve(FloatCurve=DataInterface() {propertyValues {Curve {
          defaultValue=0
          keys [
           {arriveTangent=3.2247, interpMode=RCIM_Cubic, leaveTangent=3.2247, tangentMode=RCTM_User, time=0, value=0.1}
           {arriveTangent=0.9, interpMode=RCIM_Cubic, leaveTangent=0.9, time=0.2526, value=0.6292}
           {arriveTangent=-0.0053, interpMode=RCIM_Cubic, leaveTangent=-0.0053, tangentMode=RCTM_User, time=1, value=1}
          ]
         }}}))
     Index=&SpawnGroup()
     Num=::NewEnumerator6
    )
    Scale Sprite Size Mode=::NewEnumerator2
   )
   &ScaleColor(
    Scale Alpha=&MergeHelper_FloatSwitch(
     0=0
     1=0
     2=&Multiply_Float(
      A=&FloatFromCurve(FloatCurve=DataInterface() {propertyValues {Curve {
          defaultValue=1
          keys [
           {arriveTangent=-0.1558, interpMode=RCIM_Cubic, leaveTangent=-0.1558, tangentMode=RCTM_User, time=-0.0073, value=1}
           {arriveTangent=-1.2145, interpMode=RCIM_Cubic, leaveTangent=-1.2145, tangentMode=RCTM_User, time=0.4515, value=0.2667}
           {arriveTangent=-0.1147, interpMode=RCIM_Cubic, leaveTangent=-0.1147, tangentMode=RCTM_User, time=1, value=0}
          ]
         }}})
      B=$Particles.Flicker
     )
     3=&Multiply_Float(
      A=&FloatFromCurve(FloatCurve=DataInterface() {propertyValues {Curve {
          defaultValue=1
          keys [
           {interpMode=RCIM_Cubic, tangentMode=RCTM_User, time=-0.0073, value=1}
           {arriveTangent=-1.2145, interpMode=RCIM_Cubic, leaveTangent=-1.2145, tangentMode=RCTM_User, time=0.563, value=0.7725}
           {arriveTangent=-2.6173, interpMode=RCIM_Cubic, leaveTangent=-2.6173, tangentMode=RCTM_User, time=1, value=0}
          ]
         }}})
      B=$Particles.Flicker
     )
     4=1
     5=&FloatFromCurve(FloatCurve=DataInterface() {propertyValues {Curve {
         defaultValue=1
         keys [
          {interpMode=RCIM_Cubic, tangentMode=RCTM_User, time=0.2664, value=0}
          {arriveTangent=-3.8042, interpMode=RCIM_Cubic, leaveTangent=-3.8042, tangentMode=RCTM_User, time=1, value=1}
         ]
        }}})
     6=&FloatFromCurve(FloatCurve=DataInterface() {propertyValues {Curve {
         defaultValue=1
         keys [
          {tangentMode=RCTM_User, time=0, value=1}
          {arriveTangent=-0.0405, interpMode=RCIM_Cubic, leaveTangent=-0.0405, tangentMode=RCTM_User, time=0.7009, value=1}
          {arriveTangent=-8.0089, interpMode=RCIM_Cubic, leaveTangent=-8.0089, tangentMode=RCTM_User, time=1, value=0}
         ]
        }}})
     7=0
     Index=&SpawnGroup()
     Num=::NewEnumerator6
    )
    Scale RGB=&MergeHelper_VectorSwitch(
     0=<0 0 0>
     1=<0 0 0>
     2=&MakeVectorFromLinearColorRGB(Color=&ColorFromCurve(ColorCurve=DataInterface() {propertyValues {
         AlphaCurve {keys [<at 0.2899 1 RCIM_Cubic>]}
         BlueCurve {keys [
           <at 0 9.9666>
           {arriveTangent=-43.7312, interpMode=RCIM_Cubic, leaveTangent=-43.7312, time=0.0532, value=0.2992}
           <at 0.2279 0>
           <at 0.5806 0>
           <at 0.9904 0>
          ]}
         GreenCurve {keys [
           <at 0 10>
           {arriveTangent=-42.1571, interpMode=RCIM_Cubic, leaveTangent=-42.1571, time=0.0532, value=2.4531}
           <at 0.2279 0.3922>
           <at 0.5806 0.0737>
           <at 0.9904 0.4318>
          ]}
         RedCurve {keys [
           <at 0 4.9499>
           {arriveTangent=-17.3312, interpMode=RCIM_Cubic, leaveTangent=-17.3312, time=0.0532, value=5}
           <at 0.2279 1>
           <at 0.5806 1>
           <at 0.9904 12>
          ]}
        }}))
     3=&MakeVectorFromLinearColorRGB(Color=<rgba 8 1.8933 0.4786 1>)
     4=&MakeVectorFromLinearColorRGB(Color=&ColorFromCurve(ColorCurve=DataInterface() {propertyValues {
         AlphaCurve {keys [<at 0.2041 1 RCIM_Cubic>]}
         BlueCurve {keys [
           <at 0 4.9833>
           {arriveTangent=-38.7566, interpMode=RCIM_Cubic, leaveTangent=-38.7566, time=0.025, value=0.7359}
           <at 0.1286 0>
           <at 0.3687 0>
           <at 0.9084 0.0275>
           <at 0.9981 0>
          ]}
         GreenCurve {keys [
           <at 0 5>
           {arriveTangent=-36.9042, interpMode=RCIM_Cubic, leaveTangent=-36.9042, time=0.025, value=3.2282}
           <at 0.1286 0.2549>
           <at 0.3687 0.1473>
           <at 0.9084 0.0234>
           <at 0.9981 0.4318>
          ]}
         RedCurve {keys [
           <at 0 2.4749>
           {arriveTangent=-11.471, interpMode=RCIM_Cubic, leaveTangent=-11.471, time=0.025, value=5}
           <at 0.1286 1>
           <at 0.3687 1>
           <at 0.9084 0.5781>
           <at 0.9981 12>
          ]}
        }}))
     5=&MakeVectorFromLinearColorRGB(Color=&RandomRangeLinearColor(Minimum=<rgba 10 0.5374 0 1>, Maximum=<rgba 10 6.9325 0 1>))
     6=&MakeVectorFromLinearColorRGB(Color=<rgba 20 14.5374 8.5091 1>)
     7=<0 0 0>
     Index=&SpawnGroup()
     Num=::NewEnumerator6
    )
    ScaleA=true
    ScaleRGB=true
   )
   /SetVariables(Particles.0 - Base Streaks Color:LinearColor=&ColorFromCurve(ColorCurve=DataInterface() {propertyValues {
       AlphaCurve {keys [<at 0 0.7>,<at 0.6507 0.18>,<at 1 0>]}
       BlueCurve {keys [<at 0.0424 2.6875>,<at 0.3289 0>,<at 0.6955 0>,<at 1 0>]}
       GreenCurve {keys [<at 0.0424 2.919>,<at 0.3289 1.3329>,<at 0.6955 0.3963>,<at 1 0.0724>]}
       RedCurve {keys [<at 0.0424 3>,<at 0.3289 2.5942>,<at 0.6955 1.8651>,<at 1 1>]}
      }})) {isSetParameters=true, module=SetVariables_#8}
   /SetVariables(Particles.1 - Classic Sparks Color:LinearColor=&MakeLinearColorFromVector4(Vector:Vector4f=&Multiply_Vector4(
      A=&MakeVector4_FromLinearColor(LinearColor=&ColorFromCurve(ColorCurve=DataInterface() {propertyValues {
          AlphaCurve {keys [<at 0.122 1>]}
          BlueCurve {keys [<at 0 3.0372>,<at 0.1056 0>,<at 0.2341 0>,<at 0.8178 0>,<at 0.8544 0>,<at 0.9314 0>,<at 0.9928 0>]}
          GreenCurve {keys [<at 0 7.61>,<at 0.1056 3.3715>,<at 0.2341 0.7094>,<at 0.8178 0.1839>,<at 0.8544 5.7311>,<at 0.9314 0.6262>,<at 0.9928 0.2902>]}
          RedCurve {keys [<at 0 10>,<at 0.1056 10>,<at 0.2341 7.5>,<at 0.8178 7.5>,<at 0.8544 8.0224>,<at 0.9314 10>,<at 0.9928 10>]}
         }}))
      B=&MakeVector4_FromLinearColor(LinearColor=&MakeLinearColorFromVectorAndFloat(
        "Vector (RGB)"=&Add_Vector(
         A=<1 1 1>
         B=&MakeVectorFromLinearColorHSV(Color:LinearColor=&MakeLinearColorFromHSV(
           HSV From:NiagaraInt32=1
           HSV:Vector3f=?
           Hue:NiagaraFloat=0
           Saturation:NiagaraFloat=0
           Value:NiagaraFloat=&Lerp_Float(
            A=0
            B=$Particles.Classic Flicker Value
            Alpha=&FloatFromCurve(FloatCurve=DataInterface() {propertyValues {Curve {keys [<at 0 0>,<at 0.5 0>,<at 0.55 0.9624>,<at 1 1>]}}})
           )
           Alpha:NiagaraFloat=1
          ))
        )
        "Float (Alpha)"=&Lerp_Float(
         A=1
         B=$Particles.Classic Flicker
         Alpha=&FloatFromCurve(FloatCurve=DataInterface() {propertyValues {Curve {keys [<at 0 0>,<at 0.25 0>,<at 0.35 1>,<at 1 1>]}}})
        )
       ))
     ))) {isSetParameters=true, module=SetVariables_#9}
   /SetVariables(Particles.7 - Smoke Color:LinearColor=&ColorFromCurve(ColorCurve=DataInterface() {propertyValues {
       AlphaCurve {keys [<at 0 0>,<at 0.4023 0.05>,<at 0.5672 0.1>,<at 0.7418 0.05>,<at 0.9752 0>]}
       BlueCurve {keys [<at 0 0.083>,<at 0.3354 0.0297>,<at 0.6977 0.0147>,<at 0.9899 0.0033>]}
       GreenCurve {keys [<at 0 0.0964>,<at 0.3354 0.035>,<at 0.6977 0.0053>,<at 0.9899 0.0003>]}
       RedCurve {keys [<at 0 0.2656>,<at 0.3354 0.0781>,<at 0.6977 0.0208>,<at 0.9899 0.0052>]}
      }})) {isSetParameters=true, module=SetVariables_#10}
   &BlendSpace("Position Blend (World <-> Local)":NiagaraFloat=&MakeCustomFloatFromBool(
     Bool=&SetBoolByFloatComparison(A=&VectorLength(Vector=$Engine.Owner.Velocity), B=301)
     True Float=1
     False Float=&MaskFloat(
      Mask Source=0
      Mask Parameter=?
      Mask=7
      Masked Float=&FloatFromCurve(FloatCurve=DataInterface() {propertyValues {Curve {
          defaultValue=0
          keys [{interpMode=RCIM_Cubic, tangentMode=RCTM_User, time=0, value=1}, {interpMode=RCIM_Cubic, tangentMode=RCTM_User, time=0.63, value=0}]
         }}})
      Passthrough Float=&FloatFromCurve(FloatCurve=DataInterface() {propertyValues {Curve {
          defaultValue=0
          keys [{interpMode=RCIM_Cubic, tangentMode=RCTM_User, time=0, value=1}, {interpMode=RCIM_Cubic, tangentMode=RCTM_User, time=0.7, value=0.2}]
         }}})
     )
     Translation Contribution:NiagaraFloat=1
     Rotation Contribution:NiagaraFloat=1
    )) {module=BlendSpace001}
   &CurlNoiseForce(
    Bias Noise=false
    Noise Frequency=&MaskFloat(
     Mask Source=0
     Mask Parameter=?
     Mask=7
     Masked Float=&RandomRangeFloat(Minimum=20, Maximum=120)
     Passthrough Float=0
    )
    Noise Strength=&MaskFloat(
     Mask Source=0
     Mask Parameter=?
     Mask=7
     Masked Float=&RandomRangeFloat(Minimum=300, Maximum=900)
     Passthrough Float=0
    )
    Pan Noise=false
   )
   &InheritSourceMovement(
    Applied Position Scale=&MergeHelper_VectorSwitch(
     0=<0 0 0>
     1=<0.15 0.15 0.15>
     2=<0 0 0>
     3=<0 0 0>
     4=<0 0 0>
     5=<0 0 0>
     6=<0 0 0>
     7=<0 0 0>
     Index=&SpawnGroup()
     Num=::NewEnumerator6
    )
    ApplyForce=false
    ApplyPosition=true
    ApplyPositionFalloff=false
   ) {module=InheritSourceMovement001}
   &Drag(
    Drag=&MergeHelper_FloatSwitch(
     0=0
     1=0
     2=0
     3=0
     4=0
     5=0
     6=0
     7=&RandomRangeFloat(Minimum=1, Maximum=2.5)
     Index=&SpawnGroup()
     Num=::NewEnumerator6
    )
    Use Linear Drag=true
    Use Rotational Drag=false
   )
   &SubUVAnimation(
    Output Alternate Renderer Binding=true
    Play Rate:NiagaraFloat=&RandomRangeFloat(Minimum=-0.25, Maximum=-0.5)
    Playback Scale Curve=DataInterface() {propertyValues {Curve {keys [
        {interpMode=RCIM_Cubic, tangentMode=RCTM_User, time=0.25, value=1}
        {arriveTangent=-1.2656, interpMode=RCIM_Cubic, leaveTangent=-1.2656, tangentMode=RCTM_User, time=1, value=0}
       ]}}}
    Random Start Frame=true
    "Sprite Renderer "=DataInterface() {propertyValues {SpriteRenderer=&self:Condensed_Sparks001_0.NiagaraSpriteRendererProperties_7}}
    Start Frame Offset:NiagaraInt32=&RandomRangeInt(Minimum=0, Maximum=63)
    SubUV Animation Mode:ENiagara_SubUVLookupModeV2=::NewEnumerator3
    UseEndFrame=true
    UsePlayRateScale=false
    UseStartFrame=true
   )
   &SolveForcesAndVelocity(Clamp Velocity=false, Limit Acceleration=false)
   &ScaleSpriteSizeBySpeed(
    Max Scale Factor=&MergeHelper_Vector2Switch(
     0=<1.15 6>
     1=&RandomRangeVector2D(Minimum=<0.22 3>, Maximum=<6 28>)
     2=&RandomRangeVector2D(Minimum=<1 7>, Maximum=<4 15>)
     3=&RandomRangeVector2D(Minimum=<1 5>, Maximum=<5 9>)
     4=&RandomRangeVector2D(Minimum=<1 6>, Maximum=<5 10>)
     5=<1 1>
     6=<1 5>
     7=<1 1>
     Index=&SpawnGroup()
     Num=::NewEnumerator6
    )
    Min Scale Factor=&MergeHelper_Vector2Switch(
     0=<0.75 1>
     1=<0.15 1>
     2=<0.5 2>
     3=<0.5 2>
     4=<0.5 2>
     5=<0.5 2>
     6=<0.5 2>
     7=<1 1>
     Index=&SpawnGroup()
     Num=::NewEnumerator6
    )
    Velocity Threshold:NiagaraFloat=&MergeHelper_FloatSwitch(0=1000, 1=3800, 2=200, 3=200, 4=200, 5=200, 6=200, 7=0, Index=&SpawnGroup(), Num=::NewEnumerator6)
   )
  ]
  rendererClasses [&NiagaraSpriteRendererProperties]
  renderers [
   {rendererClass=&NiagaraSpriteRendererProperties}
   {rendererClass=&NiagaraSpriteRendererProperties, rendererIndex=1}
   {rendererClass=&NiagaraSpriteRendererProperties, rendererIndex=2}
   {rendererClass=&NiagaraSpriteRendererProperties, rendererIndex=3}
   {rendererClass=&NiagaraSpriteRendererProperties, rendererIndex=4}
   {rendererClass=&NiagaraSpriteRendererProperties, rendererIndex=5}
   {rendererClass=&NiagaraSpriteRendererProperties, rendererIndex=6}
   {rendererClass=&NiagaraSpriteRendererProperties, rendererIndex=7}
  ]
 }
 Core {
  emitterUpdateScript [
   &EmitterState(Life Cycle Mode=::NewEnumerator1, Loop Behavior=::NewEnumerator1, Loop Duration=6.63, UseLoopDelay=true)
   &SpawnRate(
    SpawnRate=&FloatFromCurve(
     FloatCurve=DataInterface() {propertyValues {Curve {keys [
         {interpMode=RCIM_Cubic, tangentMode=RCTM_User, time=0, value=0}
         {interpMode=RCIM_Cubic, tangentMode=RCTM_User, time=0.0341, value=100}
         {interpMode=RCIM_Cubic, tangentMode=RCTM_User, time=0.8049, value=100.8627}
         {interpMode=RCIM_Cubic, tangentMode=RCTM_User, time=0.9514, value=-0.9412}
        ]}}}
     CurveIndex=$Emitter.NormalizedLoopAge
     Scale Curve=3
    )
    Use Spawn Probability=false
   )
  ]
  particleSpawnScript [
   &PartitionParticles()
   &InitializeParticle(
    Color Channel Mode:ENiagaraChannelCorrelation=::NewEnumerator2
    Color Minimum=<rgba 0.0448 0.0448 0.0448 1>
    Color Mode=::NewEnumerator2
    Lifetime=&RandomRangeFloat(Minimum=0.01, Maximum=0.04, Evaluation Type=::NewEnumerator1)
    Sprite Rotation Mode=::NewEnumerator2
    Sprite Size=&Vector2DFromFloat(Value=&RandomRangeFloat(Minimum=5, Maximum=20, Evaluation Type=::NewEnumerator1))
    Sprite Size Mode=::NewEnumerator4
    UsePositionOffset=false
    Write Material Random=true
   )
  ]
  particleUpdateScript [
   &ParticleState()
   &SkeletalMeshLocation(
    Bone / Socket Index=$Particles.PartitionParticles.Partition
    MaskSpawnGroup=false
    Mesh Sampling Type=::NewEnumerator5
    Skeletal Mesh=$System.Skeletal Mesh
    Socket Sampling Mode=::NewEnumerator5
   )
   &ScaleSpriteSize(
    Scale Sprite Size Mode=::NewEnumerator0
    Uniform Scale Factor=&FloatFromCurve(FloatCurve=DataInterface() {propertyValues {Curve {keys [<at 0 0>,<at 1 1>]}}})
   )
   &SubUVAnimation(
    End Frame Range Override=2
    "Sprite Renderer "=DataInterface() {propertyValues {SpriteRenderer=&self:Core_0.NiagaraSpriteRendererProperties_3}}
    UseEndFrame=false
    UseStartFrame=false
   )
   &CameraOffset(Camera Offset Amount=2)
   &ScaleColor(
    Scale RGB=&Multiply_Vector(
     A=&VectorFromFloat(Value=&MakeCustomFloatFromBool(
       Bool=$User.bIsFrontEnd
       True Float=1
       False Float=&Multiply_Float(
        A=&FNTimeOfDayScalar(Night Scale=0.75, Morning Scale=0.5, Evening Scale=0.5)
        B=&Lerp_Float(A=1, B=&RandomRangeFloat(Minimum=1.25, Maximum=2), Alpha=$User.Intensity)
       )
      ))
     B=&MakeVectorFromLinearColorRGB(Color=<rgba 10 1.5891 1.0683 1>)
    )
    ScaleA=true
    ScaleRGB=true
   )
  ]
  rendererClasses [&NiagaraSpriteRendererProperties]
  renderers [{rendererClass=&NiagaraSpriteRendererProperties}]
 }
 Flares {
  emitterUpdateScript [
   &EmitterState(Life Cycle Mode=::NewEnumerator1, Loop Behavior=::NewEnumerator1, Loop Duration=6.64, UseLoopDelay=false)
   &SpawnBurst_Instantaneous(Spawn Count=2, Use Spawn Probability=false, UseLoopCountLimit=false)
   &SpawnBurst_Instantaneous(Spawn Count=2, Spawn Group=1, Use Spawn Probability=false, UseLoopCountLimit=false) {module=SpawnBurst_Instantaneous001}
   &SpawnBurst_Instantaneous(Spawn Count=2, Spawn Group=2, Use Spawn Probability=false, UseLoopCountLimit=false) {module=SpawnBurst_Instantaneous002}
  ]
  particleSpawnScript [
   /SetVariables(Particles.Custom Random:NiagaraFloat=&RandomRangeFloat(Minimum=0.2, Maximum=0.8)) {isSetParameters=true}
   /SetVariables(Particles.VisibilityTag=&SpawnGroup()) {isSetParameters=true}
   &PartitionParticles()
   &InitializeParticle(
    Color Mode=::NewEnumerator0
    Lifetime=$Emitter.CurrentLoopDuration
    Sprite Rotation Mode=::NewEnumerator2
    Sprite Size Mode=::NewEnumerator3
    Uniform Sprite Size=&MergeHelper_FloatSwitch(0=30, 1=70, 2=25, 3=?, Index=&SpawnGroup(), Num=::NewEnumerator1)
    UsePositionOffset=false
    Write Material Random=true
   )
   &CameraOffset(Camera Offset Amount=6)
  ]
  particleUpdateScript [
   &ParticleState()
   &SpriteRotationRate(Rotation Rate=&RandomRangeFloat(Minimum=-25, Maximum=25))
   /SetVariables(Particles.Flicker:NiagaraFloat=&Waveform(
     "Waveform [1]"=::NewEnumerator0
     "[1] Frequency"=&Lerp_Float(A=4, B=6, Alpha=$Particles.Custom Random)
     "[1] Amplitude Min/Max"=&MergeHelper_Vector2Switch(0=<0 1>, 1=<0 0.3>, 2=?, 3=?, 4=?, Index=&SpawnGroup(), Num=::NewEnumerator0)
     "[1] Bias"=&RandomRangeFloat(Minimum=-0.08, Maximum=0, Evaluation Type=::NewEnumerator1)
    )) {isSetParameters=true}
   &SkeletalMeshLocation(
    Bone / Socket Index=$Particles.PartitionParticles.Partition
    MaskSpawnGroup=false
    Mesh Sampling Type=::NewEnumerator5
    Skeletal Mesh=$System.Skeletal Mesh
    Socket Sampling Mode=::NewEnumerator5
   )
   &SolveForcesAndVelocity(Clamp Velocity=false, Limit Acceleration=false)
   /SetVariables(Particles.Flare Color:LinearColor=&MakeLinearColorFromVectorAndFloat(
     "Vector (RGB)"=&Lerp_Vector(
      A=&MakeVectorFromLinearColorRGB(Color=<rgba 1 0.2274 0.0008 1>)
      B=&MakeVectorFromLinearColorRGB(Color=<rgba 1 0.0985 0 1>)
      Alpha=$Particles.Flicker
     )
     "Float (Alpha)"=&Multiply_Float(
      A=&FloatFromCurve(FloatCurve=DataInterface() {propertyValues {Curve {
          defaultValue=1
          keys [
           {arriveTangent=17.4943, interpMode=RCIM_Cubic, leaveTangent=17.4943, tangentMode=RCTM_User, time=0, value=0}
           {arriveTangentWeight=0.05, interpMode=RCIM_Cubic, leaveTangentWeight=0.2333, tangentMode=RCTM_User, time=0.15, value=1}
           {
            arriveTangent=-0.9301
            arriveTangentWeight=0.3603
            interpMode=RCIM_Cubic
            leaveTangent=-0.9301
            leaveTangentWeight=0.0772
            tangentMode=RCTM_User
            time=0.8279
            value=0.6565
           }
           {arriveTangent=-5.0054, tangentMode=RCTM_User, time=0.9828, value=0}
          ]
         }}})
      B=$Particles.Flicker
     )
    )) {isSetParameters=true}
   /SetVariables(Particles.Glow Color=&MakeLinearColorFromVectorAndFloat(
     "Vector (RGB)"=&MakeVectorFromLinearColorRGB(Color=&LerpLinearColors(StartColor=<rgba 0.2365 0.0155 0 1>, EndColor=<rgba 0.0323 0 0.0015 1>, LerpFactor=$Particles.Flicker))
     "Float (Alpha)"=&Multiply_Float(
      A=&FloatFromCurve(FloatCurve=DataInterface() {propertyValues {Curve {
          defaultValue=1
          keys [
           {arriveTangent=17.4943, interpMode=RCIM_Cubic, leaveTangent=17.4943, tangentMode=RCTM_User, time=0, value=0}
           {arriveTangentWeight=0.05, interpMode=RCIM_Cubic, leaveTangentWeight=0.2333, tangentMode=RCTM_User, time=0.15, value=1}
           {
            arriveTangent=-0.8996
            arriveTangentWeight=0.3603
            interpMode=RCIM_Cubic
            leaveTangent=-0.8996
            leaveTangentWeight=0.0772
            tangentMode=RCTM_User
            time=0.8524
            value=0.7765
           }
           {arriveTangent=-10.1516, tangentMode=RCTM_User, time=0.9863, value=0}
          ]
         }}})
      B=$Particles.Flicker
     )
    )) {isSetParameters=true}
   /SetVariables(Particles.White Flare Color:LinearColor=&MakeLinearColorFromVectorAndFloat(
     "Vector (RGB)"=&MakeVectorFromLinearColorRGB(Color=&LerpLinearColors(StartColor=<rgba 2 1.023 0 0.001>, EndColor=<rgba 2 0.0877 0.0007 0.001>, LerpFactor=$Particles.Flicker))
     "Float (Alpha)"=&Multiply_Float(
      A=&FloatFromCurve(FloatCurve=DataInterface() {propertyValues {Curve {
          defaultValue=1
          keys [
           {arriveTangent=17.4943, interpMode=RCIM_Cubic, leaveTangent=17.4943, tangentMode=RCTM_User, time=0, value=0}
           {arriveTangentWeight=0.05, interpMode=RCIM_Cubic, leaveTangentWeight=0.2333, tangentMode=RCTM_User, time=0.15, value=1}
           {
            arriveTangent=-0.7061
            arriveTangentWeight=0.3603
            interpMode=RCIM_Cubic
            leaveTangent=-0.7061
            leaveTangentWeight=0.0772
            tangentMode=RCTM_User
            time=0.7627
            value=0.7741
           }
           {arriveTangent=-6.8739, tangentMode=RCTM_User, time=0.9863, value=0}
          ]
         }}})
      B=$Particles.Flicker
     )
    )) {isSetParameters=true}
  ]
  rendererClasses [&NiagaraSpriteRendererProperties]
  renderers [
   {rendererClass=&NiagaraSpriteRendererProperties}
   {rendererClass=&NiagaraSpriteRendererProperties, rendererIndex=1}
  ]
 }
 Foot_L {
  emitterUpdateScript [
   &EmitterState(Inactive Response=::NewEnumerator1, Life Cycle Mode=::NewEnumerator1, Loop Behavior=::NewEnumerator1, Loop Duration=6.55, UseLoopDelay=false)
   &FN_SpawnAnimTrail(Number of Anim Trails:NiagaraInt32=&MakeCustomIntFromBool(
     Bool=&SetBoolByFloatComparison(A=&VectorLength(Vector=$Engine.Owner.Velocity), B=301)
     True INT=0
     False INT=1
    ))
  ]
  particleSpawnScript [
   &FN_InitializeAnimTrailBoneIndecies()
   &FN_InitializeAnimTrail(
    AnimTrailWidth=3.1
    Color=<rgba 2 0.35 0 1>
    Lifetime=&FloatFromCurve(
     FloatCurve=DataInterface() {propertyValues {Curve {
        defaultValue=1
        keys [{interpMode=RCIM_Cubic, tangentMode=RCTM_User, time=0, value=1}, {interpMode=RCIM_Cubic, tangentMode=RCTM_User, time=0.2, value=0}]
       }}}
     Scale Curve=0.25
    )
    Position Offset=<5 2 0>
    SkeletalMesh=$System.Left Foot
   )
   &VortexForce(
    Origin Pull Amount=-60
    UseOriginOffset=false
    Vortex Axis=<0 1 0>
    Vortex Force Amount=400
    Vortex Origin:NiagaraPosition=&AddVectorToPosition(Position=$Engine.Emitter.SimulationPosition, Vector=<0 0 130>)
   ) {enabled=false}
   &PointForce(
    Force Origin=&AddVectorToPosition(Position=/Niagara/DynamicInputs/Helpers/SimulationPosition(), Vector=<0 0 100>)
    Force Strength=50
    Module.Use Falloff=false
    Module.UseLinearForce=false
   ) {enabled=false}
   &ApplyInitialForces()
   &AddVelocity(Velocity=<40 0 0>) {enabled=false}
   &CameraOffset(Camera Offset Amount=10) {enabled=false}
   /SetVariables(Particles.Position=&TransformPosition(
     Position=&Multiply_Vector(A=Linked() {linkedVariable {name=Particles.Position, type=NiagaraPosition}}, B=<0 1 1>)
     Source Space=::Local
     Destination Space=::World
    )) {enabled=false, isSetParameters=true}
  ]
  particleUpdateScript [
   &ParticleState()
   &FN_UpdateAnimTrail(SkeletalMesh=$System.Left Foot)
   /SetVariables(Particles.Position=&Multiply_Vector(A=Linked() {linkedVariable {name=Particles.Position, type=NiagaraPosition}}, B=<0 1 1>)) {enabled=false, isSetParameters=true}
   &Drag(Drag=3, Use Linear Drag=true, Use Rotational Drag=true)
   &InheritVelocity(Inherited Velocity Amount Scale=<0 1 0>, Inherited Velocity Speed Limit=10, Speed Limit=true, UseSpeedThreshold=false) {enabled=false}
   &InheritSourceMovement(Applied Position Scale=<-0.2 0 -0.1>, ApplyForce=false, ApplyPosition=true, ApplyPositionFalloff=false)
   &SolveForcesAndVelocity(Clamp Velocity=false, Limit Acceleration=false)
   &ScaleRibbonWidth(Ribbon Width Scale=&FloatFromCurve(
     FloatCurve=DataInterface() {propertyValues {Curve {
        defaultValue=1
        keys [{interpMode=RCIM_Cubic, tangentMode=RCTM_User, time=0, value=0.25}, {interpMode=RCIM_Cubic, tangentMode=RCTM_User, time=1, value=1}]
       }}}
     Scale Curve=1.75
    ))
   &ScaleColor(
    Scale Mode=::NewEnumerator1
    Scale RGBA=&Multiply_Vector4(
     A=&MakeVector4_FromLinearColor(LinearColor=&ColorFromCurve(
       ColorCurve=DataInterface() {propertyValues {
         AlphaCurve {keys [<at 0.6316 1>,<at 0.8666 0.3>,<at 0.9625 0>]}
         BlueCurve {keys [<at 0 1>,<at 1 1>]}
         GreenCurve {keys [<at 0 1>,<at 1 1>]}
         RedCurve {keys [<at 0 1>,<at 1 1>]}
        }}
       CurveIndex=$Emitter.NormalizedLoopAge
      ))
     B=&MakeVector4_FromLinearColor(LinearColor=&ColorFromCurve(ColorCurve=DataInterface() {propertyValues {
         AlphaCurve {keys [<at 0.1477 1>,<at 0.4681 0.3>,<at 0.8528 0>]}
         BlueCurve {keys [<at 0 1>,<at 1 1>]}
         GreenCurve {keys [<at 0 1>,<at 1 1>]}
         RedCurve {keys [<at 0 1>,<at 1 1>]}
        }}))
    )
    ScaleRGBA=true
   ) {enabled=false}
   &ScaleColor(
    Scale Alpha=&Multiply_Float(
     A=&FloatFromCurve(FloatCurve=DataInterface() {propertyValues {Curve {keys [
          <at -0.0006 1.4995 RCIM_Cubic>
          {arriveTangent=-3.7467, interpMode=RCIM_Cubic, leaveTangent=-3.7467, tangentMode=RCTM_User, time=0.1308, value=0.5721}
          {arriveTangent=-0.0948, interpMode=RCIM_Cubic, leaveTangent=-0.0948, tangentMode=RCTM_User, time=0.4982, value=0.0589}
          <at 1.0027 0 RCIM_Cubic>
         ]}}})
     B=&FloatFromCurve(
      FloatCurve=DataInterface() {propertyValues {Curve {keys [
          {interpMode=RCIM_Cubic, tangentMode=RCTM_User, time=0.45, value=1}
          {arriveTangent=-3.035, interpMode=RCIM_Cubic, leaveTangent=-3.035, tangentMode=RCTM_User, time=0.9676, value=0}
         ]}}}
      CurveIndex=$Emitter.NormalizedLoopAge
     )
    )
    ScaleA=true
    ScaleRGB=true
   ) {module=ScaleColor001}
   /SetVariables(Particles.Position=$Particles.Initial.Position) {isSetParameters=true}
   &ScaleColor(
    Scale Alpha=&FloatFromCurve(FloatCurve=DataInterface() {propertyValues {Curve {keys [
         <at -0.0006 1.4995 RCIM_Cubic>
         {arriveTangent=-4.8957, interpMode=RCIM_Cubic, leaveTangent=-4.8957, tangentMode=RCTM_User, time=0.1896, value=0.9483}
         {arriveTangent=-0.8499, interpMode=RCIM_Cubic, leaveTangent=-0.8499, tangentMode=RCTM_User, time=0.5004, value=0.1916}
         <at 1.0027 0 RCIM_Cubic>
        ]}}})
    ScaleA=true
    ScaleRGB=true
   ) {enabled=false, module=ScaleColor002}
  ]
  rendererClasses [&NiagaraRibbonRendererProperties]
  renderers [{rendererClass=&NiagaraRibbonRendererProperties}]
 }
 Foot_R {
  emitterUpdateScript [
   &EmitterState(Inactive Response=::NewEnumerator1, Life Cycle Mode=::NewEnumerator1, Loop Behavior=::NewEnumerator1, Loop Duration=6.55, UseLoopDelay=false)
   &FN_SpawnAnimTrail(Number of Anim Trails:NiagaraInt32=&MakeCustomIntFromBool(
     Bool=&SetBoolByFloatComparison(A=&VectorLength(Vector=$Engine.Owner.Velocity), B=301)
     True INT=0
     False INT=1
    ))
  ]
  particleSpawnScript [
   &FN_InitializeAnimTrailBoneIndecies()
   &FN_InitializeAnimTrail(
    AnimTrailWidth=3.1
    Color=<rgba 2 0.35 0 1>
    Lifetime=&FloatFromCurve(
     FloatCurve=DataInterface() {propertyValues {Curve {
        defaultValue=1
        keys [{interpMode=RCIM_Cubic, tangentMode=RCTM_User, time=0, value=1}, {interpMode=RCIM_Cubic, tangentMode=RCTM_User, time=0.2, value=0}]
       }}}
     Scale Curve=0.25
    )
    Position Offset=<5 -2 0>
    SkeletalMesh=$System.Right Foot
   )
   &VortexForce(
    Origin Pull Amount=-60
    UseOriginOffset=false
    Vortex Axis=<0 1 0>
    Vortex Force Amount=400
    Vortex Origin:NiagaraPosition=&AddVectorToPosition(Position=$Engine.Emitter.SimulationPosition, Vector=<0 30 130>)
   )
   &PointForce(
    Force Origin=&AddVectorToPosition(Position=&SimulationPosition(), Vector=<0 0 100>)
    Force Strength=50
    Module.Use Falloff=false
    Module.UseLinearForce=false
   ) {enabled=false}
   &ApplyInitialForces()
   &AddVelocity(Velocity=<40 0 0>) {enabled=false}
   &CameraOffset(Camera Offset Amount=10) {enabled=false}
  ]
  particleUpdateScript [
   &ParticleState()
   &FN_UpdateAnimTrail(SkeletalMesh=$System.Right Foot)
   &Drag(Drag=3, Use Linear Drag=true, Use Rotational Drag=true)
   &InheritVelocity(Inherited Velocity Amount Scale=<0 1 0>, Inherited Velocity Speed Limit=10, Speed Limit=true, UseSpeedThreshold=false) {enabled=false}
   &InheritSourceMovement(Applied Position Scale=<-0.2 0 -0.1>, ApplyForce=false, ApplyPosition=true, ApplyPositionFalloff=false)
   &SolveForcesAndVelocity(Clamp Velocity=false, Limit Acceleration=false)
   &ScaleRibbonWidth(Ribbon Width Scale=&FloatFromCurve(
     FloatCurve=DataInterface() {propertyValues {Curve {
        defaultValue=1
        keys [{interpMode=RCIM_Cubic, tangentMode=RCTM_User, time=0, value=0.25}, {interpMode=RCIM_Cubic, tangentMode=RCTM_User, time=1, value=1}]
       }}}
     Scale Curve=1.75
    ))
   &ScaleColor(
    Scale Alpha=&Multiply_Float(
     A=&FloatFromCurve(FloatCurve=DataInterface() {propertyValues {Curve {keys [
          <at -0.0006 1.4995 RCIM_Cubic>
          {arriveTangent=-3.7467, interpMode=RCIM_Cubic, leaveTangent=-3.7467, tangentMode=RCTM_User, time=0.1308, value=0.5721}
          {arriveTangent=-0.0948, interpMode=RCIM_Cubic, leaveTangent=-0.0948, tangentMode=RCTM_User, time=0.4982, value=0.0589}
          <at 1.0027 0 RCIM_Cubic>
         ]}}})
     B=&FloatFromCurve(
      FloatCurve=DataInterface() {propertyValues {Curve {keys [
          {interpMode=RCIM_Cubic, tangentMode=RCTM_User, time=0.45, value=1}
          {arriveTangent=-3.035, interpMode=RCIM_Cubic, leaveTangent=-3.035, tangentMode=RCTM_User, time=0.9676, value=0}
         ]}}}
      CurveIndex=$Emitter.NormalizedLoopAge
     )
    )
    ScaleA=true
    ScaleRGB=true
   ) {module=ScaleColor001}
   /SetVariables(Particles.Position=$Particles.Initial.Position) {isSetParameters=true}
   &ScaleColor(
    Curve Index=$Emitter.NormalizedLoopAge
    Linear Color Curve=DataInterface() {propertyValues {
      AlphaCurve {keys [<at 0.6316 1>,<at 0.8666 0.3>,<at 0.9625 0>]}
      BlueCurve {keys [<at 0 1>,<at 1 1>]}
      GreenCurve {keys [<at 0 1>,<at 1 1>]}
      RedCurve {keys [<at 0 1>,<at 1 1>]}
     }}
    Scale Mode=::NewEnumerator2
   ) {enabled=false}
  ]
  rendererClasses [&NiagaraRibbonRendererProperties]
  renderers [{rendererClass=&NiagaraRibbonRendererProperties}]
 }
 GlowLight {
  emitterUpdateScript [
   &EmitterState(Life Cycle Mode=::NewEnumerator1, Loop Behavior=::NewEnumerator1, Loop Duration=6.64, UseLoopDelay=false)
   &SpawnBurst_Instantaneous(Use Spawn Probability=false, UseLoopCountLimit=false)
  ]
  particleSpawnScript [
   /SetVariables(Particles.Custom Random:NiagaraFloat=&RandomRangeFloat(Minimum=0.2, Maximum=0.8)) {isSetParameters=true}
   &InitializeParticle(
    Color Mode=::NewEnumerator0
    Lifetime=$Emitter.CurrentLoopDuration
    Position Offset=<0 0 120>
    Sprite Size Mode=::NewEnumerator3
    Uniform Sprite Size=215
    UsePositionOffset=true
    Write Material Random=true
   )
  ]
  particleUpdateScript [
   &ParticleState()
   /SetVariables(Particles.Flicker:NiagaraFloat=&Waveform(
     "Waveform [1]"=::NewEnumerator0
     "[1] Frequency"=&Lerp_Float(A=1, B=2, Alpha=$Particles.Custom Random)
     "[1] Amplitude Min/Max"=&MergeHelper_Vector2Switch(0=<0 1>, 1=<0 0.3>, 2=?, 3=?, 4=?, Index=&SpawnGroup(), Num=::NewEnumerator0)
     "[1] Bias"=&RandomRangeFloat(Minimum=-0.08, Maximum=0.05, Evaluation Type=::NewEnumerator1)
    )) {isSetParameters=true}
   /SetVariables(Particles.Light Glow Color:LinearColor=&MakeLinearColorFromVectorAndFloat(
     "Vector (RGB)"=&MakeVectorFromLinearColorRGB(Color=&LerpLinearColors(StartColor=<rgba 1.5 0.2405 0 0.25>, EndColor=<rgba 1.4 0.436 0 0.25>, LerpFactor=$Particles.Flicker))
     "Float (Alpha)"=&FloatFromCurve(
      FloatCurve=DataInterface() {propertyValues {Curve {
         defaultValue=0
         keys [
          {arriveTangent=4.9613, interpMode=RCIM_Cubic, leaveTangent=4.9613, tangentMode=RCTM_User, time=0, value=0}
          {
           arriveTangent=0.241
           arriveTangentWeight=0.0333
           interpMode=RCIM_Cubic
           leaveTangent=0.241
           leaveTangentWeight=0.2667
           tangentMode=RCTM_User
           time=0.1
           value=1
          }
          {
           arriveTangent=-0.1564
           arriveTangentWeight=0.3986
           interpMode=RCIM_Cubic
           leaveTangent=-0.1564
           leaveTangentWeight=0.0498
           tangentMode=RCTM_User
           time=0.9147
           value=0.9059
          }
          {arriveTangent=-5.5133, tangentMode=RCTM_User, time=1, value=0}
         ]
        }}}
      Scale Curve=0.2
     )
    )) {isSetParameters=true}
  ]
  rendererClasses [&NiagaraSpriteRendererProperties]
  renderers [{rendererClass=&NiagaraSpriteRendererProperties}]
 }
 Light {
  emitterUpdateScript [
   &EmitterState(Life Cycle Mode=::NewEnumerator1, Loop Behavior=::NewEnumerator1, Loop Duration=6.3, UseLoopDelay=false)
   &SpawnBurst_Instantaneous(Spawn Time=0.1, Use Spawn Probability=false, UseLoopCountLimit=false)
  ]
  particleSpawnScript [
   /SetVariables(Particles.Custom Random:NiagaraFloat=&RandomRangeFloat(Minimum=0.2, Maximum=0.8)) {isSetParameters=true}
   &InitializeParticle(Color Mode=::NewEnumerator0, Lifetime=6.6, Position Offset=<0 -50 120>, UsePositionOffset=true, Write Material Random=true)
   &ApplyInitialForces()
  ]
  particleUpdateScript [
   &ParticleState()
   &SkeletalMeshLocation(
    MaskSpawnGroup=false
    Mesh Sampling Type=::NewEnumerator5
    Sampled Position Offset=<0 0 4>
    Skeletal Mesh=$System.Skeletal Mesh
    Socket Sampling Mode=::NewEnumerator5
   ) {enabled=false}
   /SetVariables(Particles.Flicker:NiagaraFloat=&Waveform(
     "Waveform [1]"=::NewEnumerator0
     "[1] Frequency"=&Lerp_Float(A=1, B=2, Alpha=$Particles.Custom Random)
     "[1] Amplitude Min/Max"=&MergeHelper_Vector2Switch(0=<0 1>, 1=<0 0.3>, 2=?, 3=?, 4=?, Index=&SpawnGroup(), Num=::NewEnumerator0)
     "[1] Bias"=&RandomRangeFloat(Minimum=-0.08, Maximum=0.05, Evaluation Type=::NewEnumerator1)
    )) {isSetParameters=true}
   &FN_Light_Attributes(
    Color=&MakeLinearColorFromVectorAndFloat(
     "Vector (RGB)"=&MakeVectorFromLinearColorRGB(Color=&LerpLinearColors(StartColor=<rgba 1 0.229 0 0.15>, EndColor=<rgba 1 0.556 0 0.15>, LerpFactor=$Particles.Flicker))
     "Float (Alpha)"=&FloatFromCurve(
      FloatCurve=DataInterface() {propertyValues {Curve {
         defaultValue=0
         keys [
          {arriveTangent=4.9613, interpMode=RCIM_Cubic, leaveTangent=4.9613, tangentMode=RCTM_User, time=0, value=0}
          {
           arriveTangent=-0.0086
           arriveTangentWeight=0.0333
           interpMode=RCIM_Cubic
           leaveTangent=-0.0086
           leaveTangentWeight=0.2667
           tangentMode=RCTM_User
           time=0.1
           value=1
          }
          {arriveTangent=-0.3273, interpMode=RCIM_Cubic, leaveTangent=-0.3273, tangentMode=RCTM_User, time=0.8089, value=1}
          {arriveTangent=-0.6349, tangentMode=RCTM_User, time=0.9649, value=-0.0178}
         ]
        }}}
      Scale Curve=0.1
     )
    )
    Create Color Attribute=true
    Create Exponent Attribute=true
    Create Radius Attribute=true
    Enable Particle Light Enable=true
    Enable Visibility Tag=true
    Exponent=3
    Light Enabled=true
    Radius=100
    Scale Color=10
   )
   &ScaleColor(
    Scale Alpha=&Multiply_Float(
     A=&FloatFromCurve(
      FloatCurve=DataInterface() {propertyValues {Curve {
         defaultValue=0
         keys [
          {arriveTangent=4.9613, interpMode=RCIM_Cubic, leaveTangent=4.9613, tangentMode=RCTM_User, time=0, value=0}
          {
           arriveTangent=-0.0086
           arriveTangentWeight=0.0333
           interpMode=RCIM_Cubic
           leaveTangent=-0.0086
           leaveTangentWeight=0.2667
           tangentMode=RCTM_User
           time=0.1
           value=1
          }
          {arriveTangent=-0.3273, interpMode=RCIM_Cubic, leaveTangent=-0.3273, tangentMode=RCTM_User, time=0.5039, value=0.9323}
          {arriveTangent=-0.6349, tangentMode=RCTM_User, time=0.95, value=0}
         ]
        }}}
      Scale Curve=0.002
     )
     B=$Particles.Flicker
    )
    ScaleA=true
    ScaleRGB=true
   ) {enabled=false}
  ]
  rendererClasses [&NiagaraLightRendererProperties]
  renderers [{rendererClass=&NiagaraLightRendererProperties, rendererIndex=0}]
 }
 Other_Sparks {
  bEnabled=false
  simTarget=GPUComputeSim
  emitterUpdateScript [
   &EmitterState(Life Cycle Mode=::NewEnumerator1, Loop Behavior=::NewEnumerator1, Loop Duration=6.64, UseLoopDelay=false)
   &SpawnRate(
    SpawnRate=&FloatFromCurve(
     FloatCurve=DataInterface() {propertyValues {Curve {keys [
         {interpMode=RCIM_Cubic, tangentMode=RCTM_User, time=0, value=0}
         {arriveTangent=23.3418, interpMode=RCIM_Cubic, leaveTangent=23.3418, tangentMode=RCTM_User, time=0.0277, value=110}
         {arriveTangent=-58.9764, interpMode=RCIM_Cubic, leaveTangent=-58.9764, tangentMode=RCTM_User, time=0.7791, value=97.7763}
         {interpMode=RCIM_Cubic, tangentMode=RCTM_User, time=1, value=0}
        ]}}}
     CurveIndex=$Emitter.NormalizedLoopAge
    )
    Use Spawn Probability=false
   )
   &SpawnRate(
    SpawnGroup=1
    SpawnRate=&FloatFromCurve(
     FloatCurve=DataInterface() {propertyValues {Curve {keys [
         {arriveTangent=4983.9238, interpMode=RCIM_Cubic, leaveTangent=4983.9238, tangentMode=RCTM_User, time=0, value=0}
         {interpMode=RCIM_Cubic, tangentMode=RCTM_User, time=0.0363, value=65}
         {arriveTangent=-75.7856, interpMode=RCIM_Cubic, leaveTangent=-75.7856, tangentMode=RCTM_User, time=0.8395, value=49.9412}
         {arriveTangent=-791.5607, interpMode=RCIM_Cubic, leaveTangent=-791.5607, tangentMode=RCTM_User, time=1, value=0}
        ]}}}
     CurveIndex=$Emitter.NormalizedLoopAge
    )
    Use Spawn Probability=false
   ) {module=SpawnRate001}
   &SpawnRate(
    SpawnGroup=2
    SpawnRate=&FloatFromCurve(
     FloatCurve=DataInterface() {propertyValues {Curve {keys [
         {interpMode=RCIM_Cubic, tangentMode=RCTM_User, time=0, value=0}
         {arriveTangentWeight=0.1761, interpMode=RCIM_Cubic, leaveTangentWeight=4.9585, tangentMode=RCTM_User, time=0.0277, value=75}
         {arriveTangentWeight=25.8057, interpMode=RCIM_Cubic, leaveTangentWeight=6.3125, tangentMode=RCTM_User, time=0.8089, value=75}
         <at 1 0>
        ]}}}
     CurveIndex=$Emitter.NormalizedLoopAge
    )
    Use Spawn Probability=false
   ) {module=SpawnRate002}
   &SpawnRate(
    SpawnGroup=3
    SpawnRate=&FloatFromCurve(
     FloatCurve=DataInterface() {propertyValues {Curve {keys [
         {interpMode=RCIM_Cubic, tangentMode=RCTM_User, time=0, value=0}
         {arriveTangentWeight=0.1761, interpMode=RCIM_Cubic, leaveTangentWeight=5.2259, tangentMode=RCTM_User, time=0.0277, value=70}
         {arriveTangentWeight=27.1972, interpMode=RCIM_Cubic, leaveTangentWeight=4.921, tangentMode=RCTM_User, time=0.851, value=70}
         <at 1 0>
        ]}}}
     CurveIndex=$Emitter.NormalizedLoopAge
    )
    Use Spawn Probability=false
   ) {module=SpawnRate003}
   &SpawnRate(
    SpawnGroup=4
    SpawnRate=&FloatFromCurve(
     FloatCurve=DataInterface() {propertyValues {Curve {keys [
         {interpMode=RCIM_Cubic, tangentMode=RCTM_User, time=0, value=0}
         {arriveTangentWeight=0.1761, interpMode=RCIM_Cubic, leaveTangentWeight=5.2259, tangentMode=RCTM_User, time=0.0309, value=300}
         {arriveTangentWeight=27.1972, interpMode=RCIM_Cubic, leaveTangentWeight=4.921, tangentMode=RCTM_User, time=0.8816, value=307.5162}
         <at 0.9755 0>
        ]}}}
     CurveIndex=$Emitter.NormalizedLoopAge
    )
    Use Spawn Probability=false
   ) {module=SpawnRate004}
  ]
  particleSpawnScript [
   /SetVariables(Particles.Custom Random:NiagaraFloat=&RandomRangeFloat(Minimum=0.2, Maximum=0.8)) {isSetParameters=true}
   /SetVariables(Particles.VisibilityTag=&SpawnGroup()) {isSetParameters=true}
   &PartitionParticles()
   &InitializeParticle(
    Color Mode=::NewEnumerator0
    Lifetime=&MergeHelper_FloatSwitch(
     0=&RandomRangeFloat(Minimum=0.1, Maximum=0.2)
     1=&RandomRangeFloat(Minimum=0.2, Maximum=0.23)
     2=&RandomRangeFloat(Minimum=0.3, Maximum=0.25)
     3=&RandomRangeFloat(Minimum=0.12, Maximum=0.27)
     4=&RandomRangeFloat(Minimum=0.01, Maximum=0.05)
     Index=&SpawnGroup()
     Num=::NewEnumerator3
    )
    Sprite Size=&MergeHelper_Vector2Switch(
     0=&RandomRangeVector2D(Minimum=<0.2 0.4>, Maximum=<0.4 0.6>)
     1=&RandomRangeVector2D(Minimum=<0.5 0.6>, Maximum=<0.8 1>)
     2=<0.1 0.3>
     3=&RandomRangeVector2D(Minimum=<3 2.5>, Maximum=<8 10>)
     4=&RandomRangeVector2D(Minimum=<0.1 0.3>, Maximum=<0.3 0.6>)
     Index=&SpawnGroup()
     Num=::NewEnumerator3
    )
    Sprite Size Mode=::NewEnumerator4
    UsePositionOffset=false
    Write Material Random=true
   )
   &SkeletalMeshLocation(
    Bone / Socket Index=$Particles.PartitionParticles.Partition
    MaskSpawnGroup=false
    Mesh Sampling Type=::NewEnumerator5
    Skeletal Mesh=$System.Skeletal Mesh
    Socket Sampling Mode=::NewEnumerator5
   )
   &ShapeLocation(
    MaskSpawnGroup=false
    Override Seed=false
    Sphere Radius=&MergeHelper_FloatSwitch(0=3, 1=3.75, 2=4.25, 3=3, 4=5, Index=&SpawnGroup(), Num=::NewEnumerator3)
    UseOwnerScale=true
   )
   &CameraOffset(Camera Offset Amount=&RandomRangeFloat(Minimum=2, Maximum=12))
   &VortexForce(
    Origin Pull Amount=-30
    UseOriginOffset=false
    Vortex Axis=<1 0 0>
    Vortex Force Amount=&RandomRangeFloat(Minimum=360, Maximum=380)
    Vortex Origin:NiagaraPosition=&AddVectorToPosition(Position=$Engine.Emitter.SimulationPosition, Vector=<0 0 130>)
   )
   &ApplyInitialForces()
  ]
  particleUpdateScript [
   &ParticleState()
   /SetVariables(
    Particles.Flicker:NiagaraFloat=&Waveform("Waveform [1]"=::NewEnumerator0, "[1] Frequency"=15, "[1] Amplitude Min/Max"=<0 1>)
    Particles.Flicker Size:NiagaraFloat=&Waveform("Waveform [1]"=::NewEnumerator0, "[1] Frequency"=15, "[1] Amplitude Min/Max"=<0.25 1.5>)
    Particles.Flicker Value:NiagaraFloat=&Waveform("Waveform [1]"=::NewEnumerator0, "[1] Frequency"=15, "[1] Amplitude Min/Max"=<2 8>)
   ) {isSetParameters=true}
   &ScaleSpriteSize(
    Scale Factor=&MergeHelper_Vector2Switch(
     0=&Vector2DFromCurve(
      Vector2Curve=DataInterface() {propertyValues {
        XCurve {
         defaultValue=0
         keys [
          {arriveTangent=9.635, interpMode=RCIM_Cubic, leaveTangent=9.635, tangentMode=RCTM_User, time=0, value=1}
          {interpMode=RCIM_Cubic, tangentMode=RCTM_User, time=0.25, value=2}
          {arriveTangent=-3.6998, interpMode=RCIM_Cubic, leaveTangent=-3.6998, tangentMode=RCTM_User, time=1, value=1}
         ]
        }
        YCurve {
         defaultValue=0
         keys [
          {arriveTangent=9.635, interpMode=RCIM_Cubic, leaveTangent=9.635, tangentMode=RCTM_User, time=0, value=1}
          {interpMode=RCIM_Cubic, tangentMode=RCTM_User, time=0.25, value=2}
          {arriveTangent=-3.6998, interpMode=RCIM_Cubic, leaveTangent=-3.6998, tangentMode=RCTM_User, time=1, value=1}
         ]
        }
       }}
      ScaleCurve=&RandomRangeVector2D(Minimum=<0.5 0.5>, Maximum=<2 2.5>)
     )
     1=&Vector2DFromCurve(
      Vector2Curve=DataInterface() {propertyValues {
        XCurve {
         defaultValue=0
         keys [
          {interpMode=RCIM_Cubic, tangentMode=RCTM_User, time=0, value=1}
          {interpMode=RCIM_Cubic, tangentMode=RCTM_User, time=0.4065, value=2.4175}
          {tangentMode=RCTM_User, time=1.0157, value=1.5368}
         ]
        }
        YCurve {
         defaultValue=0
         keys [
          {interpMode=RCIM_Cubic, tangentMode=RCTM_User, time=0, value=1}
          {arriveTangent=1.2966, interpMode=RCIM_Cubic, leaveTangent=1.2966, tangentMode=RCTM_User, time=0.4288, value=1.7707}
          {arriveTangent=3.3778, tangentMode=RCTM_User, time=1.0036, value=2.7257}
         ]
        }
       }}
      ScaleCurve=&RandomRangeVector2D(Minimum=<1 1>, Maximum=<2 2.5>)
     )
     2=&Vector2DFromFloat(Value=&FloatFromCurve(
       FloatCurve=DataInterface() {propertyValues {Curve {
          defaultValue=1
          keys [
           {arriveTangent=19.7995, interpMode=RCIM_Cubic, leaveTangent=19.7995, tangentMode=RCTM_User, time=0, value=0}
           {arriveTangent=-0.1164, interpMode=RCIM_Cubic, leaveTangent=-0.1164, tangentMode=RCTM_User, time=0.098, value=1.012}
           {arriveTangent=-1.5947, interpMode=RCIM_Cubic, leaveTangent=-1.5947, tangentMode=RCTM_User, time=0.6114, value=0.6941}
           {tangentMode=RCTM_User, time=1, value=0}
          ]
         }}}
       Scale Curve=&RandomRangeFloat(Minimum=0.5, Maximum=2.5)
      ))
     3=&Vector2DFromCurve(
      Vector2Curve=DataInterface() {propertyValues {
        XCurve {keys [
          {arriveTangent=1.8483, interpMode=RCIM_Cubic, leaveTangent=1.8483, tangentMode=RCTM_User, time=0.5065, value=-0.0551}
          {
           arriveTangent=1.8483
           interpMode=RCIM_Cubic
           leaveTangent=1.8483
           leaveTangentWeight=0.3284
           tangentMode=RCTM_User
           time=0.5065
           value=-0.0551
          }
          {arriveTangent=3.2716, interpMode=RCIM_Cubic, leaveTangent=3.2716, tangentMode=RCTM_User, time=0.9778, value=1.0526}
          {
           arriveTangent=3.2716
           arriveTangentWeight=0.321
           interpMode=RCIM_Cubic
           leaveTangent=3.2716
           leaveTangentWeight=0.0074
           tangentMode=RCTM_User
           time=0.9778
           value=1.0522
          }
          {
           arriveTangent=-44.6092
           arriveTangentWeight=0.007
           interpMode=RCIM_Cubic
           leaveTangent=-44.6092
           leaveTangentWeight=0.0004
           tangentMode=RCTM_User
           time=0.9987
           value=1.0522
          }
         ]}
        YCurve {keys [
          {arriveTangent=1.8483, interpMode=RCIM_Cubic, leaveTangent=1.8483, tangentMode=RCTM_User, time=0.5065, value=-0.0551}
          {arriveTangent=3.2716, interpMode=RCIM_Cubic, leaveTangent=3.2716, tangentMode=RCTM_User, time=0.9778, value=1.0522}
          {arriveTangent=-44.6092, interpMode=RCIM_Cubic, leaveTangent=-44.6092, tangentMode=RCTM_User, time=0.9987, value=1.0522}
         ]}
       }}
      ScaleCurve=<1 1>
     )
     4=&Vector2DFromFloat(Value=&FloatFromCurve(
       FloatCurve=DataInterface() {propertyValues {Curve {keys [
           {arriveTangent=64.1395, interpMode=RCIM_Cubic, leaveTangent=64.1395, tangentMode=RCTM_User, time=-0.0046, value=1}
           {arriveTangent=-3.4207, interpMode=RCIM_Cubic, leaveTangent=-3.4207, tangentMode=RCTM_User, time=0.0506, value=3.0079}
           {arriveTangent=-2.0437, interpMode=RCIM_Cubic, leaveTangent=-2.0213, tangentMode=RCTM_Break, time=0.3795, value=1.9259}
           {interpMode=RCIM_Cubic, tangentMode=RCTM_User, time=1, value=1}
          ]}}}
       Scale Curve=1.75
      ))
     Index=&SpawnGroup()
     Num=::NewEnumerator3
    )
    Scale Sprite Size Mode=::NewEnumerator2
   )
   /SetVariables(Particles.Flicker:NiagaraFloat=&Waveform(
     "Waveform [1]"=::NewEnumerator0
     "[1] Frequency"=&Lerp_Float(A=4, B=6, Alpha=$Particles.Custom Random)
     "[1] Amplitude Min/Max"=&MergeHelper_Vector2Switch(0=<0 1>, 1=<0 0.3>, 2=?, 3=?, 4=?, Index=&SpawnGroup(), Num=::NewEnumerator0)
     "[1] Bias"=&RandomRangeFloat(Minimum=-0.08, Maximum=0, Evaluation Type=::NewEnumerator1)
    )) {isSetParameters=true}
   &BlendSpace("Position Blend (World <-> Local)":NiagaraFloat=&MakeCustomFloatFromBool(
     Bool=&SetBoolByFloatComparison(A=&VectorLength(Vector=$Engine.Owner.Velocity), B=301)
     True Float=1
     False Float=&FloatFromCurve(FloatCurve=DataInterface() {propertyValues {Curve {
         defaultValue=0
         keys [{interpMode=RCIM_Cubic, tangentMode=RCTM_User, time=0, value=1}, {interpMode=RCIM_Cubic, tangentMode=RCTM_User, time=0.7, value=0.2}]
        }}})
     Translation Contribution:NiagaraFloat=1
     Rotation Contribution:NiagaraFloat=1
    )) {module=BlendSpace001}
   &ScaleColor(
    Scale Alpha=&MergeHelper_FloatSwitch(
     0=&Multiply_Float(
      A=&FloatFromCurve(FloatCurve=DataInterface() {propertyValues {Curve {
          defaultValue=1
          keys [
           {arriveTangent=-0.1558, interpMode=RCIM_Cubic, leaveTangent=-0.1558, tangentMode=RCTM_User, time=-0.0073, value=1}
           {arriveTangent=-1.2145, interpMode=RCIM_Cubic, leaveTangent=-1.2145, tangentMode=RCTM_User, time=0.4515, value=0.2667}
           {arriveTangent=-0.1147, interpMode=RCIM_Cubic, leaveTangent=-0.1147, tangentMode=RCTM_User, time=1, value=0}
          ]
         }}})
      B=$Particles.Flicker
     )
     1=&Multiply_Float(
      A=&FloatFromCurve(FloatCurve=DataInterface() {propertyValues {Curve {
          defaultValue=1
          keys [
           {interpMode=RCIM_Cubic, tangentMode=RCTM_User, time=-0.0073, value=1}
           {arriveTangent=-1.2145, interpMode=RCIM_Cubic, leaveTangent=-1.2145, tangentMode=RCTM_User, time=0.563, value=0.7725}
           {arriveTangent=-2.6173, interpMode=RCIM_Cubic, leaveTangent=-2.6173, tangentMode=RCTM_User, time=1, value=0}
          ]
         }}})
      B=$Particles.Flicker
     )
     2=1
     3=&FloatFromCurve(FloatCurve=DataInterface() {propertyValues {Curve {
         defaultValue=1
         keys [
          {interpMode=RCIM_Cubic, tangentMode=RCTM_User, time=0.2664, value=0}
          {arriveTangent=-3.8042, interpMode=RCIM_Cubic, leaveTangent=-3.8042, tangentMode=RCTM_User, time=1, value=1}
         ]
        }}})
     4=&FloatFromCurve(FloatCurve=DataInterface() {propertyValues {Curve {
         defaultValue=1
         keys [
          {tangentMode=RCTM_User, time=0, value=1}
          {arriveTangent=-0.0405, interpMode=RCIM_Cubic, leaveTangent=-0.0405, tangentMode=RCTM_User, time=0.7009, value=1}
          {arriveTangent=-8.0089, interpMode=RCIM_Cubic, leaveTangent=-8.0089, tangentMode=RCTM_User, time=1, value=0}
         ]
        }}})
     Index=&SpawnGroup()
     Num=::NewEnumerator3
    )
    Scale RGB=&MergeHelper_VectorSwitch(
     0=&MakeVectorFromLinearColorRGB(Color=&ColorFromCurve(ColorCurve=DataInterface() {propertyValues {
         AlphaCurve {keys [<at 0.2899 1 RCIM_Cubic>]}
         BlueCurve {keys [
           <at 0 9.9666>
           {arriveTangent=-43.7312, interpMode=RCIM_Cubic, leaveTangent=-43.7312, time=0.0532, value=0.2992}
           <at 0.2279 0>
           <at 0.5806 0>
           <at 0.9904 0>
          ]}
         GreenCurve {keys [
           <at 0 10>
           {arriveTangent=-42.1571, interpMode=RCIM_Cubic, leaveTangent=-42.1571, time=0.0532, value=2.4531}
           <at 0.2279 0.3922>
           <at 0.5806 0.0737>
           <at 0.9904 0.4318>
          ]}
         RedCurve {keys [
           <at 0 4.9499>
           {arriveTangent=-17.3312, interpMode=RCIM_Cubic, leaveTangent=-17.3312, time=0.0532, value=5}
           <at 0.2279 1>
           <at 0.5806 1>
           <at 0.9904 12>
          ]}
        }}))
     1=&MakeVectorFromLinearColorRGB(Color=<rgba 8 1.8933 0.4786 1>)
     2=&MakeVectorFromLinearColorRGB(Color=&ColorFromCurve(ColorCurve=DataInterface() {propertyValues {
         AlphaCurve {keys [<at 0.2041 1 RCIM_Cubic>]}
         BlueCurve {keys [
           <at 0 4.9833>
           {arriveTangent=-38.7566, interpMode=RCIM_Cubic, leaveTangent=-38.7566, time=0.025, value=0.7359}
           <at 0.1286 0>
           <at 0.3687 0>
           <at 0.9084 0.0275>
           <at 0.9981 0>
          ]}
         GreenCurve {keys [
           <at 0 5>
           {arriveTangent=-36.9042, interpMode=RCIM_Cubic, leaveTangent=-36.9042, time=0.025, value=3.2282}
           <at 0.1286 0.2549>
           <at 0.3687 0.1473>
           <at 0.9084 0.0234>
           <at 0.9981 0.4318>
          ]}
         RedCurve {keys [
           <at 0 2.4749>
           {arriveTangent=-11.471, interpMode=RCIM_Cubic, leaveTangent=-11.471, time=0.025, value=5}
           <at 0.1286 1>
           <at 0.3687 1>
           <at 0.9084 0.5781>
           <at 0.9981 12>
          ]}
        }}))
     3=&MakeVectorFromLinearColorRGB(Color=&RandomRangeLinearColor(Minimum=<rgba 10 0.5374 0 1>, Maximum=<rgba 10 6.9325 0 1>))
     4=&MakeVectorFromLinearColorRGB(Color=<rgba 20 14.5374 8.5091 1>)
     Index=&SpawnGroup()
     Num=::NewEnumerator3
    )
    ScaleA=true
    ScaleRGB=true
   )
   &InheritSourceMovement(Applied Position Scale=<0.15 0.15 0.15>, ApplyForce=false, ApplyPosition=true, ApplyPositionFalloff=false) {module=InheritSourceMovement001}
   &CurlNoiseForce(
    Bias Noise=false
    Noise Frequency=70
    Noise Strength=&MergeHelper_FloatSwitch(0=0, 1=0, 2=0, 3=30, Index=&SpawnGroup(), Num=::NewEnumerator2)
    Pan Noise=false
   )
   &Drag(
    Drag=&FloatFromCurve(
     FloatCurve=DataInterface() {propertyValues {Curve {defaultValue=1, keys [{tangentMode=RCTM_User, time=0, value=0}, {tangentMode=RCTM_User, time=1, value=1}]}}}
     Scale Curve=&RandomRangeFloat(Minimum=1, Maximum=3)
    )
    Use Linear Drag=true
    Use Rotational Drag=true
   )
   &InheritVelocity(
    Inherited Velocity Amount Scale=&VectorFromCurve(VectorCurve=DataInterface() {propertyValues {
       XCurve {keys [
         {arriveTangent=-0.1597, interpMode=RCIM_Cubic, leaveTangent=-0.1597, tangentMode=RCTM_User, time=0.4, value=0}
         {arriveTangent=-1.0241, interpMode=RCIM_Cubic, leaveTangent=-1.0241, tangentMode=RCTM_User, time=1, value=-0.2}
        ]}
       YCurve {keys [
         {arriveTangent=-0.1597, interpMode=RCIM_Cubic, leaveTangent=-0.1597, tangentMode=RCTM_User, time=0.4, value=0}
         {arriveTangent=-1.0241, interpMode=RCIM_Cubic, leaveTangent=-1.0241, tangentMode=RCTM_User, time=1, value=-0.2}
        ]}
       ZCurve {keys [
         {arriveTangent=-0.1597, interpMode=RCIM_Cubic, leaveTangent=-0.1597, tangentMode=RCTM_User, time=0.4, value=0}
         {arriveTangent=-1.0241, interpMode=RCIM_Cubic, leaveTangent=-1.0241, tangentMode=RCTM_User, time=1, value=-0.2}
        ]}
      }})
    Inherited Velocity Speed Limit=20
    Speed Limit=true
    UseSpeedThreshold=false
   )
   &SolveForcesAndVelocity(Clamp Velocity=false, Limit Acceleration=false)
   &ScaleSpriteSizeBySpeed(
    Max Scale Factor=&MergeHelper_Vector2Switch(
     0=&RandomRangeVector2D(Minimum=<1 7>, Maximum=<4 15>)
     1=&RandomRangeVector2D(Minimum=<1 5>, Maximum=<5 9>)
     2=&RandomRangeVector2D(Minimum=<1 6>, Maximum=<5 10>)
     3=<1 1>
     4=<1 5>
     Index=&SpawnGroup()
     Num=::NewEnumerator3
    )
    Min Scale Factor=<0.5 2>
    Velocity Threshold=200
   )
  ]
  rendererClasses [&NiagaraSpriteRendererProperties]
  renderers [
   {rendererClass=&NiagaraSpriteRendererProperties}
   {rendererClass=&NiagaraSpriteRendererProperties, rendererIndex=1}
   {rendererClass=&NiagaraSpriteRendererProperties, rendererIndex=2}
   {rendererClass=&NiagaraSpriteRendererProperties, rendererIndex=3}
   {rendererClass=&NiagaraSpriteRendererProperties, rendererIndex=4}
  ]
 }
 Smoke {
  bEnabled=false
  emitterUpdateScript [
   &EmitterState(Life Cycle Mode=::NewEnumerator1, Loop Behavior=::NewEnumerator1, Loop Duration=6.64, UseLoopDelay=false)
   &SpawnRate(
    SpawnRate=&FloatFromCurve(
     FloatCurve=DataInterface() {propertyValues {Curve {
        defaultValue=1
        keys [
         {arriveTangent=78.1641, interpMode=RCIM_Cubic, leaveTangent=78.1641, tangentMode=RCTM_User, time=0.1854, value=2.0588}
         {arriveTangent=187.1201, interpMode=RCIM_Cubic, leaveTangent=187.1201, tangentMode=RCTM_User, time=0.4492, value=46.3661}
         {arriveTangent=-181.6984, interpMode=RCIM_Cubic, leaveTangent=-181.6984, tangentMode=RCTM_User, time=0.7677, value=43.7717}
         {arriveTangent=-2.3561, interpMode=RCIM_Cubic, leaveTangent=-2.3561, tangentMode=RCTM_User, time=0.973, value=0}
        ]
       }}}
     CurveIndex=$Emitter.NormalizedLoopAge
    )
    Use Spawn Probability=false
   )
  ]
  particleSpawnScript [
   &PartitionParticles()
   &InitializeParticle(
    Color=<rgba 1 1 1 0.5>
    Lifetime=&RandomRangeFloat(Minimum=0.8, Maximum=1.6)
    Sprite Size Mode=::NewEnumerator3
    Uniform Sprite Size=&RandomRangeFloat(Minimum=30, Maximum=115)
    UsePositionOffset=false
    Write Material Random=true
   )
   &SkeletalMeshLocation(
    MaskSpawnGroup=false
    Mesh Sampling Type=::NewEnumerator5
    Sampled Position Offset=<0 0 4>
    Skeletal Mesh=$System.Skeletal Mesh
    Socket Sampling Mode=::NewEnumerator5
   )
   &VortexForce(
    Origin Pull Amount=-30
    UseOriginOffset=false
    Vortex Axis=<1 0 0>
    Vortex Force Amount=&RandomRangeFloat(Minimum=75, Maximum=115)
    Vortex Origin:NiagaraPosition=&AddVectorToPosition(Position=$Engine.Emitter.SimulationPosition, Vector=<0 0 130>)
   )
   &ShapeLocation(MaskSpawnGroup=false, Override Seed=false, Sphere Radius=5, UseOwnerScale=true)
   &ApplyInitialForces()
  ]
  particleUpdateScript [
   &ParticleState()
   &ScaleSpriteSize(Uniform Curve Sprite Scale=DataInterface() {propertyValues {Curve {
       defaultValue=0
       keys [
        {arriveTangent=3.2247, interpMode=RCIM_Cubic, leaveTangent=3.2247, tangentMode=RCTM_User, time=0, value=0.1}
        {arriveTangent=0.9, interpMode=RCIM_Cubic, leaveTangent=0.9, time=0.2526, value=0.6292}
        {arriveTangent=-0.0053, interpMode=RCIM_Cubic, leaveTangent=-0.0053, tangentMode=RCTM_User, time=1, value=1}
       ]
      }}})
   &BlendSpace("Position Blend (World <-> Local)":NiagaraFloat=&MakeCustomFloatFromBool(
     Bool=&SetBoolByFloatComparison(A=&VectorLength(Vector=$Engine.Owner.Velocity), B=301)
     True Float=1
     False Float=&FloatFromCurve(FloatCurve=DataInterface() {propertyValues {Curve {
         defaultValue=0
         keys [{interpMode=RCIM_Cubic, tangentMode=RCTM_User, time=0, value=1}, {interpMode=RCIM_Cubic, tangentMode=RCTM_User, time=0.63, value=0}]
        }}})
     Translation Contribution:NiagaraFloat=1
     Rotation Contribution:NiagaraFloat=1
    )) {module=BlendSpace001}
   &ScaleColor(
    Linear Color Curve=DataInterface() {propertyValues {
      AlphaCurve {keys [<at 0 0>,<at 0.4023 0.1>,<at 0.5672 0.2>,<at 0.7418 0.1>,<at 0.9752 0>]}
      BlueCurve {keys [<at 0 0.083>,<at 0.3354 0.0297>,<at 0.6977 0.0147>,<at 0.9899 0.0033>]}
      GreenCurve {keys [<at 0 0.0964>,<at 0.3354 0.035>,<at 0.6977 0.0053>,<at 0.9899 0.0003>]}
      RedCurve {keys [<at 0 0.2656>,<at 0.3354 0.0781>,<at 0.6977 0.0208>,<at 0.9899 0.0052>]}
     }}
    Scale Mode=::NewEnumerator2
   )
   &CurlNoiseForce(
    Bias Noise=false
    Noise Frequency=&RandomRangeFloat(Minimum=20, Maximum=120)
    Noise Strength=&RandomRangeFloat(Minimum=300, Maximum=900)
    Pan Noise=false
   )
   &SubUVAnimation(
    Play Rate:NiagaraFloat=&RandomRangeFloat(Minimum=-0.25, Maximum=-0.5)
    Playback Scale Curve=DataInterface() {propertyValues {Curve {keys [
        {interpMode=RCIM_Cubic, tangentMode=RCTM_User, time=0.25, value=1}
        {arriveTangent=-1.2656, interpMode=RCIM_Cubic, leaveTangent=-1.2656, tangentMode=RCTM_User, time=1, value=0}
       ]}}}
    Random Start Frame=true
    "Sprite Renderer "=DataInterface() {propertyValues {SpriteRenderer=&self:NiagaraEmitter_7.NiagaraSpriteRendererProperties_0}}
    Start Frame Offset:NiagaraInt32=&RandomRangeInt(Minimum=0, Maximum=63)
    SubUV Animation Mode:ENiagara_SubUVLookupModeV2=::NewEnumerator3
    UseEndFrame=true
    UsePlayRateScale=false
    UseStartFrame=true
   )
   &Drag(Drag=&RandomRangeFloat(Minimum=1, Maximum=2.5), Use Linear Drag=true, Use Rotational Drag=true)
   &SolveForcesAndVelocity(Clamp Velocity=false, Limit Acceleration=false)
  ]
  rendererClasses [&NiagaraSpriteRendererProperties]
  renderers [{rendererClass=&NiagaraSpriteRendererProperties}]
 }
 Streaks {
  simTarget=GPUComputeSim
  emitterUpdateScript [
   &EmitterState(Life Cycle Mode=::NewEnumerator1, Loop Behavior=::NewEnumerator1, Loop Duration=6.64, UseLoopDelay=false)
   &SpawnRate(
    SpawnRate=&Multiply_Float(
     A=&FloatFromCurve(
      FloatCurve=DataInterface() {propertyValues {Curve {keys [{interpMode=RCIM_Cubic, tangentMode=RCTM_User, time=0.502, value=500}, {interpMode=RCIM_Cubic, tangentMode=RCTM_User, time=0.9649, value=0}]}}}
      CurveIndex=$Emitter.NormalizedLoopAge
      Scale Curve=3
     )
     B=$User.Intensity
    )
    Use Spawn Probability=false
   )
   &SpawnRate(
    SpawnGroup=1
    SpawnRate=&Multiply_Float(
     A=&FloatFromCurve(
      FloatCurve=DataInterface() {propertyValues {Curve {keys [{interpMode=RCIM_Cubic, tangentMode=RCTM_User, time=0.498, value=80}, {interpMode=RCIM_Cubic, tangentMode=RCTM_User, time=0.9609, value=0}]}}}
      CurveIndex=$Emitter.NormalizedLoopAge
      Scale Curve=3
     )
     B=$User.Intensity
    )
    Use Spawn Probability=false
   ) {module=SpawnRate001}
  ]
  particleSpawnScript [
   /SetVariables(
    Transient.ShortColor:LinearColor=&MakeLinearColorFromHSV(
     HSV From:NiagaraInt32=1
     HSV:Vector3f=?
     Hue:NiagaraFloat=&RandomRangeFloat(Minimum=0, Maximum=0.24)
     Saturation:NiagaraFloat=&RandomRangeFloat(Minimum=0, Maximum=1)
     Value:NiagaraFloat=&RandomRangeFloat(Minimum=1, Maximum=7)
     Alpha:NiagaraFloat=&RandomRangeFloat(Minimum=0.2, Maximum=1)
    )
    Transient.ShortLifetime:NiagaraFloat=&RandomRangeFloat(Minimum=0.01, Maximum=0.025, Evaluation Type=::NewEnumerator1)
    Transient.ShortScale:Vector2f=&RandomRangeVector2D(Minimum=<0.02 1>, Maximum=<0.8 8>, Evaluation Type=::NewEnumerator1)
   ) {isSetParameters=true}
   /SetVariables(
    Transient.LongColor:LinearColor=&RandomRangeLinearColor(Minimum=<rgba 0.5 0.5 0.5 1>, Maximum=<rgba 15 15 15 1>, Color Channel Mode=::NewEnumerator2)
    Transient.LongLifetime:NiagaraFloat=&RandomRangeFloat(Minimum=0.01, Maximum=0.025, Evaluation Type=::NewEnumerator1)
    Transient.LongScale:Vector2f=&RandomRangeVector2D(Minimum=<0.02 0.8>, Maximum=<0.8 1.25>, Evaluation Type=::NewEnumerator1)
   ) {isSetParameters=true}
   &PartitionParticles()
   /SetVariables(Particles.VisibilityTag=&SpawnGroup()) {isSetParameters=true}
   &InitializeParticle(
    Color=&MaskLinearColorBySpawnGroup(Masked Linear Color=$Transient.ShortColor, Passthrough Linear Color=$Transient.LongColor)
    Lifetime=&MaskFloatBySpawnGroup(Masked Float=$Transient.ShortLifetime, Passthrough Float=$Transient.LongLifetime)
    Sprite Size=&MaskVector2dBySpawnGroup(Masked Vector2D=$Transient.ShortScale, Passthrough Vector2D=$Transient.LongScale, Spawn Group Mask=0)
    Sprite Size Mode=::NewEnumerator4
    UsePositionOffset=false
    Write Material Random=true
   )
   &AddVelocity(Velocity=&RandomRangeVector(Minimum=<-250 -250 -250>, Maximum=<250 250 250>, Evaluation Type=::NewEnumerator1))
  ]
  particleUpdateScript [
   &ParticleState()
   &ScaleSpriteSize(
    Scale Factor=&Vector2DFromCurve(
     Vector2Curve=DataInterface() {propertyValues {
       XCurve {keys [<at 0 1 RCIM_Cubic>, {arriveTangent=-1.9252, interpMode=RCIM_Cubic, leaveTangent=-1.9252, tangentMode=RCTM_User, time=1, value=0.2}]}
       YCurve {keys [<at 0 3 RCIM_Cubic>, {interpMode=RCIM_Cubic, tangentMode=RCTM_User, time=1, value=0}]}
      }}
     ScaleCurve=<1 1>
    )
    Scale Sprite Size Mode=::NewEnumerator2
   )
   &AccelerationForce(Acceleration=<0 0 -980>)
   &Drag(Use Linear Drag=true, Use Rotational Drag=true)
   &SolveForcesAndVelocity(Clamp Velocity=false, Limit Acceleration=false)
   /SetVariables(
    Transient.LongColorVector:Vector3f=&MakeVectorFromLinearColorRGB(Color=&MultiplyLinearColor(
      Linear Color=&ColorFromCurve(ColorCurve=DataInterface() {propertyValues {
         AlphaCurve {keys [<at 0 1>]}
         BlueCurve {keys [<at 0.15 0.2338>,<at 0.3 1>]}
         GreenCurve {keys [<at 0.15 0.7554>,<at 0.3 1>]}
         RedCurve {keys [<at 0.15 1.25>,<at 0.3 1>]}
        }})
      Scale Factor=/Niagara/DynamicInputs/TypeConversions/MakeCustomVector4FromBool(
       Bool:NiagaraBool=$User.bIsFrontEnd
       True Vector4:Vector4f=&MakeVector4_FromLinearColor(LinearColor=$Particles.Initial.Color)
       False Vector4:Vector4f=&MakeVector4_FromFloat(Float=1)
      )
     ))
    Transient.ShortColorVector:Vector3f=&VectorFromFloat(Value=&RandomRangeFloat(Minimum=1, Maximum=1.5))
   ) {isSetParameters=true}
   &ScaleColor(
    Scale Alpha=&FNTimeOfDayScalar(Night Scale=0.5, Morning Scale=0.5, Evening Scale=0.5)
    Scale RGB=&Multiply_Vector(
     A=&MakeVectorFromLinearColorRGB(Color=<rgba 10 4.828 4.1579 1>)
     B=&MaskVectorBySpawnGroup(Masked Vector=$Transient.ShortColorVector, Passthrough Vector=$Transient.LongColorVector)
    )
    ScaleA=true
    ScaleRGB=true
   )
   /SetVariables(Particles.Position=/Niagara/DynamicInputs/Helpers/MaskPosition(
     Mask Source:NiagaraInt32=0
     Mask Parameter:NiagaraInt32=?
     Mask:NiagaraInt32=0
     Masked Position:NiagaraPosition=$Engine.Owner.Position
     Passthrough Position:NiagaraPosition=$Particles.Position
    )) {isSetParameters=true}
   &SkeletalMeshLocation(
    Bone / Socket Index=$Particles.PartitionParticles.Partition
    MaskSpawnGroup=false
    Mesh Sampling Type=::NewEnumerator5
    Skeletal Mesh=$System.Skeletal Mesh
    Socket Sampling Mode=::NewEnumerator5
   )
  ]
  rendererClasses [&NiagaraSpriteRendererProperties]
  renderers [
   {rendererClass=&NiagaraSpriteRendererProperties}
   {rendererClass=&NiagaraSpriteRendererProperties, rendererIndex=1}
  ]
 }
}
summary {
 systemName=NS_Emote_RockClimb_Sparks
 emitters [
  {emitterName=Flares, rendererClasses [&NiagaraSpriteRendererProperties]}
  {emitterName=Streaks, simTarget=GPUComputeSim, rendererClasses [&NiagaraSpriteRendererProperties]}
  {emitterName=BouncingSparks, rendererClasses [&NiagaraSpriteRendererProperties]}
  {bEnabled=false, emitterName=Smoke, rendererClasses [&NiagaraSpriteRendererProperties]}
  {emitterName=Core, rendererClasses [&NiagaraSpriteRendererProperties]}
  {emitterName=GlowLight, rendererClasses [&NiagaraSpriteRendererProperties]}
  {emitterName=Light, rendererClasses [&NiagaraLightRendererProperties]}
  {emitterName=Foot_L, rendererClasses [&NiagaraRibbonRendererProperties]}
  {emitterName=Foot_R, rendererClasses [&NiagaraRibbonRendererProperties]}
  {bEnabled=false, emitterName=Base_Streaks, rendererClasses [&NiagaraSpriteRendererProperties]}
  {bEnabled=false, emitterName=Other_Sparks, simTarget=GPUComputeSim, rendererClasses [&NiagaraSpriteRendererProperties]}
  {bEnabled=false, emitterName=Classic_Sparks, simTarget=GPUComputeSim, rendererClasses [&NiagaraSpriteRendererProperties]}
  {emitterName=Condensed_Sparks, rendererClasses [&NiagaraSpriteRendererProperties]}
  {emitterName=Condensed_Sparks001, simTarget=GPUComputeSim, rendererClasses [&NiagaraSpriteRendererProperties]}
 ]
 userVariables [
  User.bIsFrontEnd {type=NiagaraBool}
  User.bIsFrontEndPreview {type=NiagaraBool}
  User.bisHeadHIdden {type=NiagaraBool}
  User.BoneIndex {type=NiagaraFloat, defaultValue=0, description=""}
  User.ColorShift_BlueBug {type=NiagaraFloat, defaultValue=0, description=""}
  User.ForwardDot {type=NiagaraFloat, defaultValue=1, description=""}
  User.Intensity {type=NiagaraFloat, defaultValue=1, description=""}
  User.ParaGlideLeanAlpha {type=NiagaraFloat, defaultValue=0, description=""}
  User.RibbonIDOffset {type=NiagaraInt32}
 ]
}
